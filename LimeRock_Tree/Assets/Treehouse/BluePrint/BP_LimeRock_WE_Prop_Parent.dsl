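; Parent Blueprint for LimeRock treehouse props that layers an overlay mesh carrying a custom material.
; The construction script copies the static mesh from the StaticMeshComponent onto the OverlayMesh component.
; It then sets CustomOverlayMaterial on every material slot of that overlay, using a duplicate mesh as the overlay layer.
; BeginPlay, Tick and overlap events are empty, so the graphs hold no runtime logic.
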(event EventBeginPlay)

(event Collision|EventActorBeginOverlap
  (params (OtherActor :type Actor%20Object%20Reference)))

(event EventTick
  (params (DeltaSeconds :type Float%20%28single-precision%29)))

(fn ConstructionScript
  (bind _overlaymesh (Variables|Default|GetOverlayMesh))
  (Components|StaticMesh|SetStaticMesh _overlaymesh (Variables|/Script/Engine.StaticMeshComponent|GetStaticMesh (Variables|Debug|GetStaticMeshComponent)))
  (for (_ _array_index) (Rendering|Material|GetMaterials _overlaymesh)
    (Rendering|Material|SetMaterial (Variables|Default|GetOverlayMesh) _array_index (Variables|Default|GetCustomOverlayMaterial))))
