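# Placeholder material for the LimeRock treehouse asset, using masked blending, two-sided rendering and Default Lit shading.
# The graph holds only a constant green color and a sample of the T_LimeRock_Treehouse_Placeholder_M texture, with no parameters.
# Automatic setting of usage flags in the editor is turned off.
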Material /LimeRock_Tree/Assets/Treehouse/M_LimeRock_Treehouse_Placeholder
graph {
 properties {bAutomaticallySetUsageInEditor=false, blendMode=BLEND_Masked, twoSided=true, usedShadingModels=Default Lit}
 nodes [
  Constant3Vector() {constant=<rgba 0.0187 0.2656 0.0607 1>}
  TextureSample() {texture=&/T_LimeRock_Treehouse_Placeholder_M}
 ]
}
