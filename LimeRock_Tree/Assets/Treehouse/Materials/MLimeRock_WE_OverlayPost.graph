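# Translucent unlit overlay material that draws black outlines on the LimeRock Wild Estate treehouse and tree leaves.
# Opacity combines four-tap screen-space scene depth and roughness edge detection (depth only on ES3.1) with a Fresnel-biased pixel-versus-scene depth test, then fades with distance.
# WPO pushes vertices along normals, or offsets leaves toward or along the view direction with an opacity mask.
# Optional tree animation has distance-faded, Nanite-aware WPO.
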Material /LimeRock_Tree/Assets/Treehouse/Materials/MLimeRock_WE_OverlayPost
graph {
 properties {
  bAutomaticallySetUsageInEditor=false
  bUseMaterialAttributes=true
  blendMode=BLEND_Translucent
  shadingModel=MSM_Unlit
  usedShadingModels=Unlit
 }
 nodes [
  n1 = NamedRerouteUsage()
  n2 = Subtract(A=CameraPositionWS(), B=@n1)
  n3 = CollectionParameter() {
   collection=&FortniteMaterialParameters
   parameterName=SunlightDirectionalVector
  }
  n4 = ShadowReplace(Default=Normalize(VectorInput=@n2), Shadow=ComponentMask(None=@n3) {b=true, g=true, r=true})
  n5 = Normalize(VectorInput=VertexNormalWS())
  n6 = DotProduct(A=@n4, B=@n5)
  n7 = StaticSwitchParameter(True=Multiply(A=@n4, B=Sign(None=@n6)), False=@n4) {defaultValue=true, group=Billboard, parameterName=BillBoardShortestPathRotation}
  n8 = CrossProduct(A=@n5, B=@n7)
  n9 = DotProduct(A=@n7, B=@n5)
  n10 = MaterialFunctionCall(Multiplier=Constant() {r=0.5}) {materialFunction=&Pi}
  n11 = WorldPosition()
  n12 = RotateAboutAxis(
   NormalizedRotationAxis=Normalize(VectorInput=@n8)
   RotationAngle=Multiply(A=Arccosine!(None=@n9), B=Divide(A=@n10.Result) {constB=10})
   PivotPoint=@n1
   Position=@n11.XYZ
  )
  n13 = Add(A=NamedRerouteUsage(), B=NamedRerouteDeclaration(None=@n12) {name=BillboardTransformFallback, nodeColor=<rgba 1 0.1641 0 1>})
  n14 = Subtract(A=Add(A=@n13, B=WorldPosition().XYZ), B=NamedRerouteUsage())
  Add(A=@n13, B=Multiply(A=Normalize(VectorInput=@n14), B=ScalarParameter() {defaultValue=3, parameterName=ScaleBillboardOutline}))
  n15 = WorldPosition() {worldPositionShaderOffset=WPT_CameraRelativeNoOffsets}
  n16 = StaticSwitchParameter(True=Add(A=@n15.XYZ, B=NamedRerouteUsage()), False=@n15.XYZ) {group=Billboard, parameterName=UseFallbackBillboard}
  n17 = NamedRerouteUsage()
  n18 = ScalarParameter() {defaultValue=150, group=Color, parameterName=LeafVisualMergeDistance}
  n19 = Round(None=Divide(A=NamedRerouteUsage(), B=@n18))
  n20 = TransformPosition(None=NamedRerouteDeclaration(None=Multiply(A=@n19, B=@n18)) {name=LeafPivot_LocalSpace_Merged, nodeColor=<rgba 0 1 0.5547 1>}) {transformSourceType=TRANSFORMPOSSOURCE_Instance, transformType=TRANSFORMPOSSOURCE_TranslatedWorld}
  n21 = Normalize(VectorInput=@n20)
  n22 = DotProduct(A=Subtract(A=Add(A=@n16, B=@n17), B=@n20), B=@n21)
  n23 = Max(
   A=Subtract(A=Multiply(A=@n22) {constB=-1}, B=ScalarParameter() {defaultValue=66.7177, group=BillBoard WPO, parameterName=BillboardSqueze_Clamp})
   B=Constant()
  )
  Ceil(None=@n23)
  Constant3Vector() {constant=<rgba 1 1 0 0>}
  Constant() {r=1}
  n24 = SceneTexture() {sceneTextureId=PPI_Roughness}
  n25 = ComponentMask(None=ScreenPosition().ViewportUV) {g=true, r=true}
  n26 = Multiply(A=ScalarParameter() {defaultValue=2, parameterName=Simple LineThickness}, B=SceneTexelSize())
  n27 = MaterialFunctionCall!(OffsetAmount=@n26) {materialFunction=/Engine/Functions/Engine_MaterialFunctions01/ImageAdjustment/BlurSampleOffsets}
  n28 = Add(A=@n25, B=@n27."-.5,.5") {desc=x+}
  n29 = Subtract(A=@n24.Color, B=SceneTexture(UVs=@n28) {sceneTextureId=PPI_Roughness}.Color)
  n30 = Add(A=@n25, B=@n27.".5,.5") {desc=y+}
  n31 = Subtract(A=@n24.Color, B=SceneTexture(UVs=@n30) {sceneTextureId=PPI_Roughness}.Color)
  n32 = Add(A=@n25, B=@n27.".5,-.5") {desc=x-}
  n33 = Subtract(A=@n24.Color, B=SceneTexture(UVs=@n32) {sceneTextureId=PPI_Roughness}.Color)
  n34 = Add(A=@n25, B=@n27."-.5,-.5") {desc=y-}
  n35 = Subtract(A=@n24.Color, B=SceneTexture(UVs=@n34) {sceneTextureId=PPI_Roughness}.Color)
  n36 = Add(A=Add(A=Abs(None=@n29), B=Abs(None=@n31)), B=Add(A=Abs(None=@n33), B=Abs(None=@n35)))
  n37 = ScalarParameter() {defaultValue=2633.8608, group=RoughnessSampling, parameterName=Roughness_OutlineDivide}
  Divide(A=@n36, B=@n37)
  DotProduct()
  n38 = ScalarParameter() {defaultValue=1, group=Mobile, parameterName=ScaleSample_Mobile}
  n39 = ScalarParameter() {defaultValue=1, group=Mobile_Step2, parameterName=ScaleSample_Mobile_Step2}
  n40 = SceneTexture() {sceneTextureId=PPI_SceneDepth}
  n41 = ScalarParameter() {defaultValue=40000, group=SteppedTransition_Mobile, parameterName=Step1_Mobile_End}
  n42 = Saturate(None=Divide(
    A=Subtract(A=@n40.Color, B=@n41)
    B=ScalarParameter() {defaultValue=40000, group=SteppedTransition_Mobile, parameterName=Step1_Mobile_FadeDistance}
   ))
  n43 = SceneTexture() {sceneTextureId=PPI_SceneDepth}
  n44 = ScalarParameter() {defaultValue=40000, group=SteppedTransition_Mobile, parameterName=Step2_Mobile_End}
  n45 = Saturate(None=Divide(
    A=Subtract(A=@n43.Color, B=@n44)
    B=ScalarParameter() {defaultValue=40000, group=SteppedTransition_Mobile, parameterName=Step2_Mobile_FadeDistance}
   ))
  FeatureLevelSwitch(
   Default=NamedRerouteUsage()
   ES3_1=LinearInterpolate(
    A=LinearInterpolate(A=@n38, B=@n39, Alpha=@n42)
    B=ScalarParameter() {defaultValue=1, group=Mobile_Step3, parameterName=ScaleSample_Mobile_Step3}
    Alpha=@n45
   )
  )
  n46 = ScalarParameter() {defaultValue=1, group=Mobile, parameterName=BaseSampleBias_Mobile}
  n47 = ScalarParameter() {defaultValue=1, group=Mobile_Step2, parameterName=BaseSampleBias_Mobile_Step2}
  FeatureLevelSwitch(
   Default=NamedRerouteUsage()
   ES3_1=LinearInterpolate(
    A=LinearInterpolate(A=@n46, B=@n47, Alpha=@n42)
    B=ScalarParameter() {defaultValue=1, group=Mobile_Step3, parameterName=BaseSampleBias_Mobile_Step3}
    Alpha=@n45
   )
  )
  n48 = ScalarParameter() {defaultValue=1, group=Mobile, parameterName=DivisorDepthSample_Mobile}
  n49 = ScalarParameter() {defaultValue=1, group=Mobile_Step2, parameterName=DivisorDepthSample_Mobile_Step2}
  FeatureLevelSwitch(
   Default=NamedRerouteUsage()
   ES3_1=LinearInterpolate(
    A=LinearInterpolate(A=@n48, B=@n49, Alpha=@n42)
    B=ScalarParameter() {defaultValue=1, group=Mobile_Step3, parameterName=DivisorDepthSample_Mobile_Step3}
    Alpha=@n45
   )
  )
  FeatureLevelSwitch(Default=@n26, ES3_1=Multiply(A=ScalarParameter() {defaultValue=2, group=Mobile, parameterName=LineThicknessMobile}, B=SceneTexelSize()))
  MaterialFunctionCall(Vector=@n2) {materialFunction=&SafeNormalize}
  MaterialFunctionCall(Vector=@n8) {materialFunction=&SafeNormalize}
  MaterialFunctionCall(VectorToTransform=@n11.XYZ, BasisPosition=@n1) {materialFunction=&Transform3x3Matrix}
  Multiply(A=NamedRerouteUsage(), B=Divide(A=ViewSize(), B=NamedRerouteUsage()))
  n50 = ScalarParameter() {defaultValue=10, parameterName=ThresholdAlwaysDepth}
  n51 = ScalarParameter() {defaultValue=490.1322, parameterName=ThresholdAlwaysDepth2}
  n52 = DotProduct(A=VertexNormalWS(), B=Normalize(VectorInput=WorldPosition() {worldPositionShaderOffset=WPT_CameraRelative}.XYZ))
  n53 = Saturate(None=LinearInterpolate(
    A=ScalarParameter() {defaultValue=0.5765, parameterName=FresnelTheshold_low}
    B=ScalarParameter() {defaultValue=5.5104, parameterName=FresnelTheshold_High}
    Alpha=Multiply(A=@n52) {constB=-1}
   ))
  n54 = Subtract(A=Add(A=PixelDepth().R, B=LinearInterpolate(A=@n50, B=@n51, Alpha=@n53)), B=SceneDepth().R)
  n55 = OneMinus(None=Saturate(None=Multiply(A=@n36, B=@n37)))
  n56 = OneMinus(None=NamedRerouteDeclaration(None=ComponentMask(None=@n55) {r=true}) {name=Roughness_Outlines, nodeColor=<rgba 1 0 0.8438 1>})
  n57 = MaterialFunctionCall!(
   Base=SceneTexture() {sceneTextureId=PPI_SceneDepth}.Color
   BaseSampleBias=ScalarParameter() {defaultValue=1, group=Mobile_Step3, parameterName=BaseSampleBias}
   Scale=ScalarParameter() {defaultValue=2, group=Mobile, parameterName=ScaleSample}
   Sample1=SceneTexture(UVs=@n28) {sceneTextureId=PPI_SceneDepth}.Color
   Sample2=SceneTexture(UVs=@n30) {sceneTextureId=PPI_SceneDepth}.Color
   Sample3=SceneTexture(UVs=@n32) {sceneTextureId=PPI_SceneDepth}.Color
   Sample4=SceneTexture(UVs=@n34) {sceneTextureId=PPI_SceneDepth}.Color
   Divisor=ScalarParameter() {defaultValue=800, group=Mobile, parameterName=DivisorDepthSample}
   UseCenteredLines=Constant() {r=1}
   UseRGB
  ) {materialFunction=/Game/Packages/jl_tests/combine4Samples}
  n58 = Add(A=OneMinus(None=Saturate(None=@n54)), B=FeatureLevelSwitch(Default=Add(A=@n56, B=@n57.Result), ES3_1=@n57.Result))
  Multiply(A=@n58, B=Saturate(None=TwoSidedSign()))
  Multiply(B=@n53)
  n59 = Multiply(A=ScalarParameter() {defaultValue=-0.99, group=BillBoard WPO, parameterName=BillboardSqueze}, B=Multiply(A=@n21, B=@n22) {constB=-1})
  n60 = Multiply(A=@n21, B=ScalarParameter() {defaultValue=-100, group=BillBoard WPO, parameterName=BillboardOffsetCamDir})
  NamedRerouteDeclaration(None=ShadowReplace(Default=Add(A=@n59, B=@n60), Shadow=Constant3Vector())) {name=WPO_LeafFlatten, nodeColor=<rgba 0 1 0.4844 1>}
  n61 = ScalarParameter() {defaultValue=1, group=DepthSampling, parameterName=ScaleSample_1}
  n62 = ScalarParameter() {defaultValue=1, group=DepthSampling_2, parameterName=ScaleSample_2}
  n63 = NamedRerouteUsage()
  NamedRerouteDeclaration(None=LinearInterpolate(
    A=LinearInterpolate(A=@n61, B=@n62, Alpha=@n63)
    B=ScalarParameter() {defaultValue=1, group=DepthSampling_3, parameterName=ScaleSample_3}
    Alpha=NamedRerouteUsage()
   )) {name=ScaleSamples, nodeColor=<rgba 0 1 0.2969 1>}
  n64 = ScalarParameter() {defaultValue=1, parameterName=Opacity_Near}
  n65 = ScalarParameter() {defaultValue=0.5, parameterName=Opacity_Far}
  n66 = PixelDepth()
  n67 = ScalarParameter() {defaultValue=100, parameterName=Opacity_Fade_Offset}
  n68 = Saturate(None=Divide(A=Subtract(A=@n66.R, B=@n67), B=ScalarParameter() {defaultValue=20057.9883, parameterName=Opacity_Fade_BlendDist}))
  n69 = TextureSampleParameter2D() {
   parameterName=OpacityMask
   samplerType=SAMPLERTYPE_Masks
   texture=/LimeRock_Tree/Assets/Treehouse/Textures/T_WildEstate_Treehouse_Fallback_opacity
  }
  n70 = Multiply(A=Ceil(None=@n69.R), B=@n58)
  n71 = NamedRerouteUsage()
  n72 = Multiply(
   A=Normalize(VectorInput=WorldPosition() {worldPositionShaderOffset=WPT_CameraRelative}.XYZ)
   B=ScalarParameter() {defaultValue=0.5, parameterName=PushCameraDirection}
  )
  n73 = VertexNormalWS()
  n74 = ScalarParameter() {defaultValue=1.0798, parameterName=Push}
  n75 = SetMaterialAttributes!(
   MaterialAttributes
   Emissive Color=Constant3Vector()
   Opacity=Multiply(A=LinearInterpolate(A=@n64, B=@n65, Alpha=@n68), B=StaticSwitchParameter(True=@n70, False=@n58) {parameterName=IsTreeLeaves})
   World Position Offset=StaticSwitchParameter(True=Add(A=@n71, B=@n72), False=Multiply(A=@n73, B=@n74)) {parameterName=IsTreeLeaves}
  ) {attributeSetTypes [#1, #2, #3]}
  n76 = MaterialFunctionCall!(In=@n75, OpaqueCanopyMask=Constant3Vector()) {materialFunction=/LimeRock_Tree/Assets/Treehouse/MaterialFunctions/MF_LimeRock_WE_ApolloTreeAnim}
  NamedRerouteDeclaration(None=@n76.Leaf_Pivot_LS) {name=LeafPivot_LocalSpace, nodeColor=<rgba 0.8047 0 1 1>}
  NamedRerouteDeclaration(None=@n76.WPO_OverallBend) {name=Anim_WPO_OverallBend, nodeColor=<rgba 0.8047 0 1 1>}
  n77 = ScalarParameter() {defaultValue=1, group=DepthSampling, parameterName=BaseSampleBias_1}
  n78 = ScalarParameter() {defaultValue=1, group=DepthSampling_2, parameterName=BaseSampleBias_2}
  n79 = NamedRerouteUsage()
  NamedRerouteDeclaration(None=LinearInterpolate(
    A=LinearInterpolate(A=@n77, B=@n78, Alpha=@n79)
    B=ScalarParameter() {defaultValue=1, group=DepthSampling_3, parameterName=BaseSampleBias_3}
    Alpha=NamedRerouteUsage()
   )) {name=BaseSampleBias, nodeColor=<rgba 1 0 0.0234 1>}
  NamedRerouteDeclaration(None=@n13) {name=WPO BillboardTransformFallback, nodeColor=<rgba 1 0.1641 0 1>}
  n80 = ScalarParameter() {defaultValue=-800, group=DepthSampling, parameterName=DivisorDepthSamples}
  n81 = ScalarParameter() {defaultValue=-800, group=DepthSampling_2, parameterName=DivisorDepthSamples2}
  n82 = NamedRerouteUsage()
  NamedRerouteDeclaration(None=LinearInterpolate(
    A=LinearInterpolate(A=@n80, B=@n81, Alpha=@n82)
    B=ScalarParameter() {defaultValue=-800, group=DepthSampling_3, parameterName=DivisorDepthSamples3}
    Alpha=NamedRerouteUsage()
   )) {name=DivisoDepthSamples, nodeColor=<rgba 0 1 0.6953 1>}
  NamedRerouteDeclaration(None=@n76.Leaf_Pivot_WS) {name=LeafPivot_FallBack_WorldSpace, nodeColor=<rgba 0.8047 0 1 1>}
  NamedRerouteDeclaration(None=@n76.Leaf_XAxis_WS) {name=Leaf XAxis_ WorldSpace, nodeColor=<rgba 0.8047 0 1 1>}
  NamedRerouteDeclaration(None=@n76.Anim WPO_Only) {name=Anim_WPO_only, nodeColor=<rgba 1 0 0.3984 1>}
  n83 = ScalarParameter() {defaultValue=1, group=DepthSampling, parameterName=LineThickness_1}
  n84 = ScalarParameter() {defaultValue=1, group=DepthSampling_2, parameterName=LineThickness_2}
  n85 = NamedRerouteUsage()
  NamedRerouteDeclaration(None=LinearInterpolate(
    A=LinearInterpolate(A=@n83, B=@n84, Alpha=@n85)
    B=ScalarParameter() {defaultValue=1, group=DepthSampling_3, parameterName=LineThickness_3}
    Alpha=NamedRerouteUsage()
   )) {name=LineThickness, nodeColor=<rgba 1 0.7734 0 1>}
  n86 = Subtract(
   A=SceneTexture() {sceneTextureId=PPI_SceneDepth}.Color
   B=ScalarParameter() {defaultValue=2000, group=SteppedTransition, parameterName=Step1_End}
  )
  n87 = ScalarParameter() {defaultValue=1, group=SteppedTransition, parameterName=Step1_FadeDistance}
  NamedRerouteDeclaration(None=Saturate(None=Divide(A=@n86, B=@n87))) {name=Step1Blend, nodeColor=<rgba 0.4531 0 1 1>}
  n88 = ScalarParameter() {defaultValue=2000, group=DepthSampling, parameterName=BiasLineThickness_1}
  n89 = ScalarParameter() {defaultValue=2000, group=DepthSampling_2, parameterName=BiasLineThickness_2}
  n90 = NamedRerouteUsage()
  NamedRerouteDeclaration(None=LinearInterpolate(
    A=LinearInterpolate(A=@n88, B=@n89, Alpha=@n90)
    B=ScalarParameter() {defaultValue=2000, group=DepthSampling_3, parameterName=BiasLineThickness_3}
    Alpha=NamedRerouteUsage()
   )) {name=BaseLineThickness, nodeColor=<rgba 1 0.4453 0 1>}
  n91 = Subtract(
   A=SceneTexture() {sceneTextureId=PPI_SceneDepth}.Color
   B=ScalarParameter() {defaultValue=40000, group=SteppedTransition, parameterName=Step2_End}
  )
  n92 = ScalarParameter() {defaultValue=40000, group=SteppedTransition, parameterName=Step2_FadeDistance}
  NamedRerouteDeclaration(None=Saturate(None=Divide(A=@n91, B=@n92))) {name=Step2Blend, nodeColor=<rgba 0.4531 0 1 1>}
  NamedRerouteUsage()
  NamedRerouteUsage()
  ObjectPositionWS()
  n93 = Multiply(A=@n22) {constB=-1}
  n94 = ScalarParameter() {defaultValue=104.7052, group=BillBoard WPO, parameterName=Billboard_EdgeMask}
  Saturate(None=Divide(A=Subtract(A=@n93, B=@n94)) {constB=0.0001})
  SceneDepth()
  SceneTexture() {sceneTextureId=PPI_Roughness}
  ShadowReplace(Default=@n12, Shadow=Constant3Vector())
  n95 = GetMaterialAttributes(None=@n76.Result) {attributeGetTypes [#3]}
  n96 = ScalarParameter() {defaultValue=1500, group=WPO Fade, parameterName=WPO_Fade_Distance}
  n97 = Divide(A=Subtract(A=PixelDepth().R, B=ScalarParameter() {defaultValue=7500, group=WPO Fade, parameterName=WPO_Fade_Offset}), B=@n96)
  n98 = MaterialFunctionCall!(FadeBlendDistance=@n96, AutoHandleBadDistance) {materialFunction=/Game/Global/MaterialFunctions/TechArt/MF_WPODisableBlend}
  n99 = LinearInterpolate(A=@n95.World Position Offset, B=Constant3Vector(), Alpha=Saturate(None=NaniteReplace!(Default=@n97, Nanite=@n98.VertexShaderOutput)))
  StaticSwitchParameter(
   True=StaticSwitchParameter(
    True=SetMaterialAttributes!(MaterialAttributes=@n95.MaterialAttributes, World Position Offset=@n99) {attributeSetTypes [#3]}
    False=@n76.Result
   ) {defaultValue=true, group=WPO Fade, parameterName=UseWPOFade}
   False=@n75
  ) {parameterName=UseTreeAnimation}
  n100 = Multiply(A=@n21, B=Saturate(None=@n23))
  n101 = ScalarParameter() {defaultValue=-206.2373, group=BillBoard WPO, parameterName=BillboardSqueze_PushInner}
  StaticSwitchParameter(True=@n59, False=Add(A=@n59, B=Multiply(A=@n100, B=@n101))) {defaultValue=true, parameterName=UseVirtual Plane Card}
  StaticSwitchParameter() {parameterName=IsTreeLeaves}
  n102 = TextureCoordinate() {coordinateIndex=2}
  n103 = AppendVector(
   A=ComponentMask(None=TextureCoordinate() {coordinateIndex=1}) {r=true}
   B=AppendVector(A=ComponentMask(None=@n102) {r=true}, B=ComponentMask(None=@n102) {g=true})
  )
  Subtract(A=Multiply(A=Frac(None=@n103)) {constB=12000}, B=Constant3Vector() {constant=<rgba 6000 6000 3000 0>}) {constB=6000}
  TransformPosition(None=NamedRerouteDeclaration(None=NamedRerouteUsage()) {name=LeafPivot Position LS, nodeColor=<rgba 0.0781 1 0 1>}) {transformType=TRANSFORMPOSSOURCE_World}
  Transform(None=@n17) {transformSourceType=TRANSFORMSOURCE_World, transformType=TRANSFORM_Camera}
  VertexNormalWS()
 ]
}
