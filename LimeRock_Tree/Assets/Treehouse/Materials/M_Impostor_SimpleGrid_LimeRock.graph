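# Masked two-sided subsurface impostor material for LimeRock trees, drawing camera-facing sprites from a baked hemi-octahedral frame atlas.
# View direction selects three frames, blended with depth parallax from normal-map alpha, which also drives pixel depth offset.
# Shadow-pass switches substitute light direction for camera inputs, and mobile overrides use a single-frame version.
# Switchable legacy, Apollo and Asteria color systems add trunk masking, stepped sun shading and world, instance or primitive-data variation.
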Material /LimeRock_Tree/Assets/Treehouse/Materials/M_Impostor_SimpleGrid_LimeRock
graph {
 properties {
  bAlwaysEvaluateWorldPositionOffset=true
  bAutomaticallySetUsageInEditor=false
  bTangentSpaceNormal=false
  bUseMaterialAttributes=true
  bUsedWithInstancedStaticMeshes=true
  blendMode=BLEND_Masked
  shadingModel=MSM_Subsurface
  twoSided=true
  usedShadingModels=Subsurface
 }
 nodes [
  n1 = Custom() {code="View.ViewToClip[3][3]", desc=Ignore pixel depth offset is camera is ortho, outputType=CMOT_Float1, inputs []}
  Abs(None=Subtract(B=Step(X=@n1)))
  n2 = Constant3Vector()
  n3 = MaterialFunctionCall(Float3=Multiply(
    A=TransformPosition(None=@n2) {transformType=TRANSFORMPOSSOURCE_World}
    B=ScalarParameter() {defaultValue=0.0012, group=Apollo, parameterName=Color Variation Range WS, sortPriority=10}
   ) {constB=0.0012}) {materialFunction=&BreakOutFloat3Components}
  n4 = Add(A=Multiply(A=@n3.R) {constB=0.5}, B=Multiply(A=@n3.G) {constB=0.5})
  n5 = Frac(None=@n3.B)
  Add(A=Multiply(A=Subtract(A=@n4) {constB=0.5}) {constB=2}, B=Multiply(A=Subtract(A=@n5) {constB=0.5}) {constB=2})
  MaterialExpressionCameraPositionWS_0 = CameraPositionWS()
  MaterialExpressionNamedRerouteUsage_100 = NamedRerouteUsage()
  n6 = Custom!(Main=@MaterialExpressionCameraPositionWS_0, Shadow=@MaterialExpressionNamedRerouteUsage_100) {
   code="##if SHADOW_DEPTH_SHADER \r\n{\r\nreturn Shadow;\r\n}\r\n##endif\r\nreturn Main;"
   description=Shadow Pass Switch
   inputs [
    {
     inputName=Main
     input {expression=&self:MaterialExpressionCameraPositionWS_0}
    }
    {
     inputName=Shadow
     input {expression=&self:MaterialExpressionNamedRerouteUsage_100}
    }
   ]
  }
  n7 = MaterialFunctionCall() {materialFunction=/Game/ImpostorBaker/MaterialFunctions/ImpostorInfo}
  n8 = Normalize(VectorInput=Transform(None=Subtract(A=@n6, B=@n7.Position)) {transformSourceType=TRANSFORMSOURCE_World, transformType=TRANSFORM_Tangent})
  n9 = MaterialFunctionCall(Float3=@n8) {materialFunction=&BreakOutFloat3Components}
  n10 = Normalize(VectorInput=MaterialFunctionCall(X=@n9.R, Y=@n9.R, Z=Max(A=@n9.B) {constB=0.001}) {materialFunction=&MakeFloat3}.Result)
  n11 = Clamp(None=ConstantBiasScale(None=MaterialFunctionCall!(Unit Vector=@n10) {materialFunction=/Game/ImpostorBaker/MaterialFunctions/UnitVectorToHemiOctahedra}.Hemi Octahedron))
  n12 = Subtract(A=ScalarParameter() {defaultValue=16, parameterName=FramesXY})
  n13 = Floor(None=Multiply(A=@n11, B=@n12))
  n14 = Multiply(A=TextureCoordinate(), B=@n12)
  n15 = Floor(None=@n14)
  n16 = AppendVector(A=ComponentMask(None=@n13) {r=true}, B=ComponentMask(None=@n15) {r=true})
  n17 = Frac(None=@n14)
  n18 = MaterialFunctionCall(UVs=@n17, Texture Scale=ScalarParameter() {defaultValue=1, parameterName=UVscale}) {materialFunction=&ScaleUVsByCenter}
  AppendVector(
   A=Multiply(
    A=ScalarParameter() {defaultValue=1, parameterName=G}
    B=MaterialFunctionCall!(Vector2=@n16, MaximumNumberOfDigits, UVs=@n18.UVs, DebugTextLocation RG_UpperRight BA_LowerLeft, Component Spacing) {materialFunction=/Engine/Functions/Engine_MaterialFunctions02/Utility/DebugFloat2Values}.ColorCodedOutput
   )
   B=Constant()
  )
  CameraPositionWS()
  CameraPositionWS()
  CameraPositionWS()
  CameraPositionWS()
  CameraPositionWS()
  Constant3Vector()
  Constant()
  Custom!(InWorldVector) {
   code="\r\n##if USE_INSTANCING || IS_MESHPARTICLE_FACTORY\r\n\t\tfloat3 temp;\r\n\t\ttemp.x = length(TransformLocalVectorToWorld(Parameters, float3(1,0,0)));\r\n\t\ttemp.y = length(TransformLocalVectorToWorld(Parameters, float3(0,1,0)));\r\n\t\ttemp.z = length(TransformLocalVectorToWorld(Parameters, float3(0,0,1)));\r\n\t\treturn mul(InWorldVector, (MaterialFloat3x3)transpose(Parameters.InstanceLocalToWorld)) / (temp*temp);\r\n\t##else\r\n\t\treturn mul(InWorldVector, (MaterialFloat3x3)Primitive.WorldToLocal);\r\n\t##endif"
   description=World to Loc
   inputs [{inputName=InWorldVector, input {expression=null}}]
  }
  Custom() {
   code="\r\n##if USE_INSTANCING || IS_MESHPARTICLE_FACTORY\r\n\t\tfloat3 temp;\r\n\t\ttemp.x = length(TransformLocalVectorToWorld(Parameters, float3(1,0,0)));\r\n\t\ttemp.y = length(TransformLocalVectorToWorld(Parameters, float3(0,1,0)));\r\n\t\ttemp.z = length(TransformLocalVectorToWorld(Parameters, float3(0,0,1)));\r\n\t\treturn mul(InWorldVector, (MaterialFloat3x3)transpose(Parameters.InstanceLocalToWorld)) / (temp*temp);\r\n\t##else\r\n\t\treturn mul(InWorldVector, (MaterialFloat3x3)Primitive.WorldToLocal);\r\n\t##endif"
   description=World to Loc
   inputs [{inputName=InWorldVector, input {expression=null}}]
  }
  n19 = NamedRerouteUsage()
  Divide(A=@n19, B=VectorParameter() {parameterName=TrunkExtractColor, defaultValue=<rgba 1 1 1 1>}.RGB)
  n20 = Constant() {r=0.5}
  n21 = MaterialFunctionCall(Float3=WorldPosition() {desc=Imposter is tiny and matches almost actor location, worldPositionShaderOffset=WPT_ExcludeAllShaderOffsets}.XYZ) {materialFunction=&BreakOutFloat3Components}
  n22 = ScalarParameter() {defaultValue=13000, group=World Hue Variation, parameterName=Color Variation Distance WS, sortPriority=10}
  n23 = ScalarParameter() {defaultValue=11000, group=World Hue Variation, parameterName=Color Variation Distance WS2, sortPriority=10}
  n24 = Add(A=Sine(None=Divide(A=@n21.R, B=@n22)), B=Sine(None=Divide(A=@n21.G, B=@n23)))
  n25 = Divide(A=Add(A=@n24) {constB=2}) {constB=4}
  n26 = StaticSwitchParameter(True=@n25, False=ScalarParameter() {defaultValue=0.5, group=Apollo, parameterName=Fixed Color Manual Offset, sortPriority=-9}) {group=Apollo, parameterName=UseRandomColor, sortPriority=-10}
  n27 = TextureObjectParameter() {group=Apollo, parameterName=ColorPalette, sortPriority=99, texture=/Game/Environments/Apollo/Foliage/ColorLUT/Apollo_Foliage_LUT_Green}
  n28 = TextureSample(UVs=AppendVector(A=FeatureLevelSwitch(Default=@n26, ES2=@n20, ES3_1=@n20, SM4_REMOVED=@n26), B=Constant() {r=0.375}), Tex=@n27) {samplerSource=SSM_Wrap_WorldGroupSettings, texture=&Apollo_Foliage_LUT_Green}
  n29 = AppendVector(A=@n26, B=Constant() {r=0.125})
  n30 = CollectionParameter() {
   collection=&FortniteMaterialParameters
   parameterName=SunlightDirectionalVector
  }
  n31 = Saturate(None=DotProduct(A=ComponentMask(None=@n30) {b=true, g=true, r=true}, B=NamedRerouteUsage()))
  n32 = Saturate(None=Subtract(A=@n31, B=Constant() {r=0.5}))
  n33 = Constant() {r=0.625}
  n34 = Constant() {r=-0.5}
  n35 = OneMinus(None=Saturate(None=Subtract(A=@n31, B=@n34)))
  n36 = LinearInterpolate(
   A=LinearInterpolate(
    A=@n28.RGB
    B=TextureSample(UVs=@n29, Tex=@n27) {samplerSource=SSM_Wrap_WorldGroupSettings, texture=&Apollo_Foliage_LUT_Green}.RGB
    Alpha=@n32
   )
   B=TextureSample(UVs=AppendVector(A=@n26, B=@n33), Tex=@n27) {samplerSource=SSM_Wrap_WorldGroupSettings, texture=&Apollo_Foliage_LUT_Green}.RGB
   Alpha=@n35
  )
  FeatureLevelSwitch(Default=@n36, ES2=@n28.RGB, ES3_1=@n28.RGB, SM4_REMOVED=@n36)
  n37 = Multiply(A=@n11, B=Subtract(A=ScalarParameter() {defaultValue=16, parameterName=FramesXY}))
  n38 = VertexInterpolator(VS=@n37)
  n39 = MaterialFunctionCall!(UV=Frac(None=@n38.PS)) {materialFunction=/Game/ImpostorBaker/MaterialFunctions/TriangleInterpolator}
  LinearInterpolate(
   A=VectorParameter() {parameterName=Manual Weights, defaultValue=<rgba 0 0 1 1>}.RGB
   B=@n39.Result
   Alpha=ScalarParameter() {defaultValue=1, parameterName=Enable Weights}
  )
  n40 = MaterialFunctionCall() {materialFunction=&ImpostorInfo}
  n41 = ScalarParameter() {defaultValue=1, parameterName=WPOScale}
  n42 = MaterialFunctionCall() {materialFunction=&ImpostorInfo}
  MaterialExpressionCameraPositionWS_9 = CameraPositionWS()
  MaterialExpressionNamedRerouteUsage_101 = NamedRerouteUsage()
  n43 = Custom!(Main=@MaterialExpressionCameraPositionWS_9, Shadow=@MaterialExpressionNamedRerouteUsage_101) {
   code="##if SHADOW_DEPTH_SHADER \r\n{\r\nreturn Shadow;\r\n}\r\n##endif\r\nreturn Main;"
   description=Shadow Pass Switch
   inputs [
    {
     inputName=Main
     input {expression=&self:MaterialExpressionCameraPositionWS_9}
    }
    {
     inputName=Shadow
     input {expression=&self:MaterialExpressionNamedRerouteUsage_101}
    }
   ]
  }
  MaterialExpressionMaterialFunctionCall_91 = MaterialFunctionCall() {materialFunction=&CameraDirectionVector}
  MaterialExpressionNamedRerouteUsage_104 = NamedRerouteUsage()
  n44 = Custom(Main=@MaterialExpressionMaterialFunctionCall_91.Result, Shadow=@MaterialExpressionNamedRerouteUsage_104) {
   code="##if SHADOW_DEPTH_SHADER \r\n{\r\nreturn Shadow;\r\n}\r\n##endif\r\nreturn Main;"
   description=Shadow Pass Switch
   inputs [
    {
     inputName=Main
     input {expression=&self:MaterialExpressionMaterialFunctionCall_91}
    }
    {
     inputName=Shadow
     input {expression=&self:MaterialExpressionNamedRerouteUsage_104}
    }
   ]
  }
  n45 = DotProduct(A=Normalize(VectorInput=Subtract(A=@n42.Position, B=@n43)), B=@n44)
  n46 = MaterialFunctionCall!(
   Center Location=@n40.Position
   0-1 UVs=Multiply(A=TextureCoordinate() {uTiling=10, vTiling=10}, B=ScalarParameter() {defaultValue=16, parameterName=FramesXY})
   XY scale=Multiply(A=@n40.Scaled Size, B=LinearInterpolate(A=@n41, Alpha=@n45))
   OverridePivot
   "Normals (Optional)"
  ) {materialFunction=&Sprite}
  n47 = Add(A=WorldPosition() {worldPositionShaderOffset=WPT_ExcludeAllShaderOffsets}.XYZ, B=Multiply(A=@n46.World position offset))
  n48 = Normalize(VectorInput=Transform(None=Subtract(A=@n6, B=@n47)) {transformSourceType=TRANSFORMSOURCE_World, transformType=TRANSFORM_Tangent})
  n49 = Distance(A=@n43, B=MaterialFunctionCall() {materialFunction=&ImpostorInfo}.Position)
  n50 = Multiply(A=MaterialFunctionCall() {materialFunction=&ImpostorInfo}.Scaled Size) {constB=0.25}
  n51 = Saturate(None=Divide(A=Subtract(A=@n49, B=@n50), B=@n50)) {desc=Don't Push Verts into Camera}
  LinearInterpolate(A=@n8, B=@n48, Alpha=Multiply(A=ScalarParameter() {defaultValue=1, parameterName=InterpolatePosition}, B=@n51))
  n52 = MaterialFunctionCall() {materialFunction=&ImpostorInfo}
  n53 = Add(A=@n46.World position offset, B=WorldPosition().XYZ)
  n54 = Subtract(A=@n43, B=@n53)
  n55 = Normalize(VectorInput=@n54) {desc=Vector to Geo}
  MaterialExpressionMaterialFunctionCall_89 = MaterialFunctionCall() {materialFunction=&CameraDirectionVector}
  MaterialExpressionNamedRerouteUsage_105 = NamedRerouteUsage()
  n56 = Custom(Main=@MaterialExpressionMaterialFunctionCall_89.Result, Shadow=@MaterialExpressionNamedRerouteUsage_105) {
   code="##if SHADOW_DEPTH_SHADER \r\n{\r\nreturn Shadow;\r\n}\r\n##endif\r\nreturn Main;"
   description=Shadow Pass Switch
   inputs [
    {
     inputName=Main
     input {expression=&self:MaterialExpressionMaterialFunctionCall_89}
    }
    {
     inputName=Shadow
     input {expression=&self:MaterialExpressionNamedRerouteUsage_105}
    }
   ]
  }
  n57 = Multiply(
   A=Subtract(
    A=Distance(A=@n43, B=@n52.Position)
    B=MaterialFunctionCall(Vector 3=@n54) {desc=Dist to Geo, materialFunction=/Engine/Functions/Engine_MaterialFunctions02/Utility/VectorLength}.V3 Length
   )
   B=Abs(None=DotProduct(A=@n55, B=@n56))
  )
  n58 = Normalize(VectorInput=Subtract(A=@n43, B=@n53)) {desc=CameraVector}
  n59 = Multiply(A=MaterialFunctionCall() {materialFunction=&ImpostorInfo}.Scaled Size) {constB=0.5}
  n60 = Add(
   A=Multiply(A=Multiply(A=@n57, B=@n55), B=ScalarParameter() {defaultValue=-1, parameterName=Center Align})
   B=Multiply(A=Multiply(A=@n58, B=@n59), B=ScalarParameter() {defaultValue=1, parameterName=Camera Push})
  )
  n61 = DistanceFieldsRenderingSwitch!(No=Constant(), Yes=@n60)
  MaterialExpressionFeatureLevelSwitch_6 = FeatureLevelSwitch(Default=@n60, ES2=@n61, ES3_1=@n61, SM4_REMOVED=@n60, SM5=@n60, SM6=@n60)
  MaterialExpressionConstant_45 = Constant()
  n62 = Custom(Main=@MaterialExpressionFeatureLevelSwitch_6, Shadow=@MaterialExpressionConstant_45) {
   code="##if SHADOW_DEPTH_SHADER \r\n{\r\nreturn Shadow;\r\n}\r\n##endif\r\nreturn Main;"
   description=Shadow Pass Switch
   inputs [
    {
     inputName=Main
     input {expression=&self:MaterialExpressionFeatureLevelSwitch_6}
    }
    {inputName=Shadow, input {expression=&self:MaterialExpressionConstant_45}}
   ]
  }
  n63 = Add(A=@n46.World position offset, B=@n62) {constB=0}
  LinearInterpolate(A=@n63, B=@n63, Alpha=@n51)
  n64 = MaterialFunctionCall() {materialFunction=&ImpostorInfo}
  LinearInterpolate(A=@n64.Position, B=@n47, Alpha=@n51)
  n65 = ScalarParameter() {bUseCustomPrimitiveData=true, defaultValue=0.5, group=Color, parameterName=ValueVariation_primdata, primitiveDataIndex=5}
  LinearInterpolate(
   A=ScalarParameter() {defaultValue=0.5, group=Color Variation WorldSpace, parameterName=ValueVariation}
   B=StaticSwitchParameter(True=PerInstanceCustomData!(DefaultValue) {constDefaultValue=0.5, dataIndex=5}, False=@n65) {defaultValue=true, parameterName=UsePerInstanceCustomData}
   Alpha=NamedRerouteUsage()
  )
  n66 = ScalarParameter() {bUseCustomPrimitiveData=true, defaultValue=-0.03, group=Color, parameterName=Hue Variation_primdata, primitiveDataIndex=4}
  LinearInterpolate(A=ScalarParameter() {defaultValue=-0.03, group=Color Variation WorldSpace, parameterName=Hue Variation}, B=@n66, Alpha=NamedRerouteUsage())
  MaterialFunctionCall!(
   ImposterPosition=MaterialFunctionCall() {materialFunction=&ImpostorInfo}.OutputBasePosition
   WPO_Input
   In
  ) {materialFunction=/Game/Athena/Apollo/Environments/Landscape/Materials/Biome/UniqueBiomeMaterialFunctions/MF_Apollo_WaterLevel}
  n67 = ComponentMask(None=CollectionParameter() {parameterName=Direction}) {b=true, g=true, r=true}
  Multiply(A=@n67) {constB=-1}
  Multiply(A=@n63, B=ScalarParameter() {defaultValue=0.9503})
  n68 = DotProduct(A=@n67, B=CameraVectorWS()) {desc=L dot camera}
  n69 = Power(
   Base=Max(A=Saturate(None=@n68), B=ScalarParameter() {defaultValue=0.3, parameterName=ScatterMin})
   Exp=ScalarParameter() {defaultValue=2, parameterName=fresnel}
  )
  n70 = OneMinus(None=ComponentMask(None=TextureCoordinate()) {g=true})
  n71 = NamedRerouteUsage()
  n72 = VectorParameter() {parameterName=Color, defaultValue=<rgba 1 0.4475 0.0933 5>}
  n73 = Multiply(A=@n72.RGB, B=@n72.RGB)
  Multiply(A=NamedRerouteUsage(), B=Multiply(A=Multiply(A=@n69, B=@n70), B=Multiply(A=@n71, B=@n73)))
  n74 = Multiply(A=TextureCoordinate(), B=Subtract(A=ScalarParameter() {defaultValue=16, parameterName=FramesXY}))
  Multiply(
   A=ScalarParameter() {defaultValue=1, parameterName=Debug Grid}
   B=MaterialFunctionCall!(UV=Frac(None=@n74), Line Width=ScalarParameter() {defaultValue=0.03, parameterName=grid line}) {materialFunction=/Game/ImpostorBaker/MaterialFunctions/TriangulatedWireframe}.Result
  )
  n75 = Divide(A=Floor(None=@n38.PS), B=ScalarParameter() {defaultValue=16, parameterName=FramesXY})
  n76 = Multiply(A=@n48)
  n77 = Floor(None=@n37)
  n78 = MaterialFunctionCall!(
   InVec=@n76
   X Vec
   Y Vec
   Z Vec=MaterialFunctionCall!(Frame=@n77, XY Frames=ScalarParameter() {defaultValue=16, parameterName=FramesXY}) {materialFunction=/Game/ImpostorBaker/MaterialFunctions/ImposterFrameTransform_Setup}.World Z
  ) {materialFunction=/Game/ImpostorBaker/MaterialFunctions/ImposterFrameTransform}
  n79 = Multiply(
   A=MaterialFunctionCall() {materialFunction=&ImpostorInfo}.Scaled Size
   B=ScalarParameter() {defaultValue=1, parameterName=TanScaleFix}
  )
  MaterialExpressionCameraPositionWS_10 = CameraPositionWS() {desc=Ray Origin}
  MaterialExpressionNamedRerouteUsage_102 = NamedRerouteUsage()
  n80 = Custom(Main=@MaterialExpressionCameraPositionWS_10, Shadow=@MaterialExpressionNamedRerouteUsage_102) {
   code="##if SHADOW_DEPTH_SHADER \r\n{\r\nreturn Shadow;\r\n}\r\n##endif\r\nreturn Main;"
   description=Shadow Pass Switch
   inputs [
    {
     inputName=Main
     input {expression=&self:MaterialExpressionCameraPositionWS_10}
    }
    {
     inputName=Shadow
     input {expression=&self:MaterialExpressionNamedRerouteUsage_102}
    }
   ]
  }
  n81 = Transform(None=Subtract(A=@n47, B=@n80)) {transformSourceType=TRANSFORMSOURCE_World, transformType=TRANSFORM_Tangent}
  n82 = Transform(None=Subtract(A=@n80, B=@n64.Position)) {transformSourceType=TRANSFORMSOURCE_World, transformType=TRANSFORM_Tangent}
  n83 = VertexInterpolator(VS=MaterialFunctionCall!(
    Plane Normal Axis=@n78.World Z
    Plane X Axis=@n78.World Z
    Plane Y Axis=Multiply(A=@n78.World Y) {constB=-1}
    UV Scale=@n79
    Ray Dir Local=@n81
    Ray Origin Local=@n82
   ) {materialFunction=/Game/ImpostorBaker/MaterialFunctions/VirtualPlaneCoordinates_Imposter}.UVs)
  n84 = Add(A=@n75, B=Divide(A=Saturate(None=@n83.PS), B=ScalarParameter() {defaultValue=16, parameterName=FramesXY}))
  MaterialExpressionReroute_30 = Reroute(None=@n84)
  n85 = Subtract(B=TextureSampleParameter2D(UVs=@n84) {constMipValue=0, parameterName=Normal, samplerSource=SSM_Wrap_WorldGroupSettings, texture=&Black}.A) {constA=0.5, constB=0.5}
  n86 = VertexInterpolator(VS=Divide(A=ComponentMask(None=@n78.Local Z Vector) {g=true, r=true}, B=ScalarParameter() {defaultValue=16, parameterName=FramesXY}) {constB=16})
  n87 = Constant() {r=1}
  MaterialExpressionAdd_42 = Add(A=@MaterialExpressionReroute_30, B=Multiply(A=Multiply(A=@n85, B=@n86.PS), B=@n87))
  n88 = LinearInterpolate(A=Constant2Vector() {g=1}, B=Constant2Vector() {r=1}, Alpha=@n39.Mask)
  n89 = ScalarParameter() {defaultValue=16, parameterName=FramesXY}
  n90 = LinearInterpolate(
   A=Constant2Vector() {g=1}
   B=Constant2Vector() {r=1}
   Alpha=MaterialFunctionCall(UV=Frac(None=@n37)) {materialFunction=&TriangleInterpolator}.Mask
  )
  n91 = MaterialFunctionCall(
   InVec=@n76
   Z Vec=MaterialFunctionCall(Frame=Add(A=@n77, B=@n90), XY Frames=ScalarParameter() {defaultValue=16, parameterName=FramesXY}) {materialFunction=&ImposterFrameTransform_Setup}.World Z
  ) {materialFunction=&ImposterFrameTransform}
  n92 = VertexInterpolator(VS=MaterialFunctionCall(
    Plane Normal Axis=@n91.World Z
    Plane X Axis=@n91.World Z
    Plane Y Axis=Multiply(A=@n91.World Y) {constB=-1}
    UV Scale=@n79
    Ray Dir Local=@n81
    Ray Origin Local=@n82
   ) {materialFunction=&VirtualPlaneCoordinates_Imposter}.UVs)
  n93 = Add(A=Add(A=@n75, B=Divide(A=@n88, B=@n89)), B=Divide(A=Saturate(None=@n92.PS), B=ScalarParameter() {defaultValue=16, parameterName=FramesXY}))
  n94 = Subtract(B=TextureSampleParameter2D(UVs=@n93) {constMipValue=0, parameterName=Normal, samplerSource=SSM_Wrap_WorldGroupSettings, texture=&Black}.A) {constA=0.5, constB=0.5}
  n95 = VertexInterpolator(VS=Divide(A=ComponentMask(None=@n91.Local Z Vector) {g=true, r=true}, B=ScalarParameter() {defaultValue=16, parameterName=FramesXY}) {constB=16})
  n96 = Add(A=@n93, B=Multiply(A=Multiply(A=@n94, B=@n95.PS), B=@n87))
  n97 = Constant() {r=1}
  n98 = ScalarParameter() {defaultValue=16, parameterName=FramesXY}
  n99 = Constant2Vector() {g=1, r=1}
  n100 = MaterialFunctionCall(
   InVec=@n76
   Z Vec=MaterialFunctionCall(Frame=Add(A=@n77, B=@n99), XY Frames=ScalarParameter() {defaultValue=16, parameterName=FramesXY}) {materialFunction=&ImposterFrameTransform_Setup}.World Z
  ) {materialFunction=&ImposterFrameTransform}
  n101 = VertexInterpolator(VS=MaterialFunctionCall(
    Plane Normal Axis=@n100.World Z
    Plane X Axis=@n100.World Z
    Plane Y Axis=Multiply(A=@n100.World Y) {constB=-1}
    UV Scale=@n79
    Ray Dir Local=@n81
    Ray Origin Local=@n82
   ) {materialFunction=&VirtualPlaneCoordinates_Imposter}.UVs)
  n102 = Add(A=Add(A=@n75, B=Divide(A=@n97, B=@n98)), B=Divide(A=Saturate(None=@n101.PS), B=ScalarParameter() {defaultValue=16, parameterName=FramesXY}))
  n103 = Subtract(B=TextureSampleParameter2D(UVs=@n102) {constMipValue=0, parameterName=Normal, samplerSource=SSM_Wrap_WorldGroupSettings, texture=&Black}.A) {constA=0.5, constB=0.5}
  n104 = VertexInterpolator(VS=Divide(A=ComponentMask(None=@n100.Local Z Vector) {g=true, r=true}, B=ScalarParameter() {defaultValue=16, parameterName=FramesXY}) {constB=16})
  n105 = Add(A=@n102, B=Multiply(A=Multiply(A=@n103, B=@n104.PS), B=@n87))
  n106 = Multiply(A=@n39.Result)
  n107 = Multiply(A=TextureCoordinate() {uTiling=10, vTiling=10}, B=ScalarParameter() {defaultValue=0.6, parameterName=Mip UV Scale})
  n108 = MaterialFunctionCall!(
   Texture Object=TextureObjectParameter() {parameterName=Normal, texture=&Black}
   R UV=@MaterialExpressionAdd_42
   G UV=@n96
   B UV=@n105
   Weights=@n106
   Derivative UVs=@n107
  ) {materialFunction=/Game/ImpostorBaker/MaterialFunctions/FrameBlendWeights_TextureObject}
  n109 = ComponentMask(None=@n108.Result) {b=true, g=true, r=true}
  n110 = MaterialFunctionCall() {materialFunction=/Game/ImpostorBaker/MaterialFunctions/Impostor_SingleFrameVersion}
  n111 = Transform(None=ConstantBiasScale(None=MaterialFunctionCall!(Default=@n109, Mobile Override=@n110.Normal RGB) {materialFunction=/Game/ImpostorBaker/MaterialFunctions/OverrideMobile}.Result) {bias=-0.5, scale=2})
  NamedRerouteDeclaration(None=@n111) {name=WorldSpaceNormal, nodeColor=<rgba 1 0.75 0 1>}
  n112 = NamedRerouteUsage()
  n113 = CollectionParameter() {
   collection=&FortniteMaterialParameters
   parameterName=SunlightDirectionalVector
  }
  n114 = Add(A=DotProduct(A=Normalize(VectorInput=@n112), B=ComponentMask(None=@n113) {b=true, g=true, r=true}))
  n115 = ScalarParameter() {
   defaultValue=2
   desc="Shading Range on the Dot product 2 = full range. From -1 to 1 of the dot product.\r\nlower value makes the shaded area smaller \r\nhigher value makes the shaded area larger"
   group=Color
   parameterName=Shading_Range
  }
  n116 = ScalarParameter() {
   defaultValue=8
   desc="Shading Range on the Dot product 2 = full range. From -1 to 1 of the dot product.\r\nlower value makes the shaded area smaller \r\nhigher value makes the shaded area larger"
   group=Color
   parameterName=Shading_Steps
  }
  n117 = Round(None=Multiply(A=Divide(A=@n114, B=@n115), B=@n116))
  n118 = Power(
   Base=LinearInterpolate(
    A=ScalarParameter() {defaultValue=0.512, group=Color, parameterName=Shading_min}
    B=ScalarParameter() {defaultValue=1.1151, group=Color, parameterName=Shading_Max}
    Alpha=Divide(A=@n117, B=@n116)
   )
   Exp=ScalarParameter() {defaultValue=2.6, group=Color, parameterName=Shading_ImposotorMatchTweak}
  )
  n119 = Multiply(A=ScalarParameter() {defaultValue=0.104, group=Color, parameterName=RandomHueShift}, B=Subtract(A=NamedRerouteUsage()) {constB=0.5})
  n120 = NamedRerouteUsage()
  NamedRerouteDeclaration(None=Multiply(
    A=Saturate(None=@n118)
    B=MaterialFunctionCall(Hue Shift Percentage=@n119, Texture=@n120) {materialFunction=&HueShift}.Result
   )) {name=BaseColorFinal, nodeColor=<rgba 0.2422 1 0 1>}
  n121 = Divide(
   A=Subtract(A=ComponentMask(None=@n43) {b=true}, B=ScalarParameter() {defaultValue=10000, parameterName=Camera Z Fadeout Loc})
   B=ScalarParameter() {defaultValue=35000, parameterName=Cam Z Fade Length}
  )
  n122 = Add(A=@n110.WPO, B=Multiply(A=VectorParameter() {parameterName=MobileZOffset, defaultValue=<rgba 0 0 550 0>}.RGB, B=Saturate(None=@n121)))
  NamedRerouteDeclaration(None=MaterialFunctionCall(Default=@n63, Mobile Override=DistanceFieldsRenderingSwitch(No=@n122, Yes=@n63)) {materialFunction=&OverrideMobile}.Result) {name=WPO Final, nodeColor=<rgba 0 0.1094 1 1>}
  n123 = MaterialFunctionCall(Default=ComponentMask(None=@n108.Result) {a=true}, Mobile Override=@n110.Normal A) {materialFunction=&OverrideMobile}
  n124 = ScalarParameter() {defaultValue=350, parameterName=Default Mesh Size}
  n125 = MaterialFunctionCall() {materialFunction=/Engine/Functions/Engine_MaterialFunctions02/WorldPositionOffset/ObjectScale}
  n126 = ScalarParameter() {defaultValue=1, parameterName=PDO}
  MaterialExpressionCameraVectorWS_4 = CameraVectorWS()
  MaterialExpressionNamedRerouteUsage_103 = NamedRerouteUsage()
  n127 = Custom(Main=@MaterialExpressionCameraVectorWS_4, Shadow=@MaterialExpressionNamedRerouteUsage_103) {
   code="##if SHADOW_DEPTH_SHADER \r\n{\r\nreturn Shadow;\r\n}\r\n##endif\r\nreturn Main;"
   description=Shadow Pass Switch
   inputs [
    {
     inputName=Main
     input {expression=&self:MaterialExpressionCameraVectorWS_4}
    }
    {
     inputName=Shadow
     input {expression=&self:MaterialExpressionNamedRerouteUsage_103}
    }
   ]
  }
  MaterialExpressionMaterialFunctionCall_83 = MaterialFunctionCall() {materialFunction=&CameraDirectionVector}
  n128 = Custom(Main=@MaterialExpressionMaterialFunctionCall_83.Result, Shadow=@MaterialExpressionNamedRerouteUsage_103) {
   code="##if SHADOW_DEPTH_SHADER \r\n{\r\nreturn Shadow;\r\n}\r\n##endif\r\nreturn Main;"
   description=Shadow Pass Switch
   inputs [
    {
     inputName=Main
     input {expression=&self:MaterialExpressionMaterialFunctionCall_83}
    }
    {
     inputName=Shadow
     input {expression=&self:MaterialExpressionNamedRerouteUsage_103}
    }
   ]
  }
  n129 = Abs(None=DotProduct(A=Normalize(VectorInput=@n127), B=@n128))
  n130 = Multiply(
   A=OneMinus(None=ComponentMask(None=@n123.Result) {r=true})
   B=Multiply(A=Multiply(A=@n124, B=@n125.Scale X), B=Multiply(A=@n126, B=@n129)) {constB=0.5}
  )
  n131 = Constant()
  n132 = FeatureLevelSwitch(Default=@n130, ES2=@n131, ES3_1=@n131, SM4_REMOVED=@n130)
  n133 = Constant()
  n134 = OneMinus(None=@n1)
  NamedRerouteDeclaration(None=Multiply(A=QualitySwitch(Default=@n132, Low=@n133, Medium=@n133), B=Saturate(None=@n134))) {name=PixelDepthOffset, nodeColor=<rgba 0 0.8828 1 1>}
  n135 = ComponentMask(None=NamedRerouteUsage()) {b=true, g=true, r=true}
  n136 = Multiply(A=Desaturation(None=@n135, Fraction=Constant() {r=1})) {constB=5}
  n137 = Custom(Main=@MaterialExpressionAdd_42, Shadow=@MaterialExpressionReroute_30) {
   code="##if SHADOW_DEPTH_SHADER \r\n{\r\nreturn Shadow;\r\n}\r\n##endif\r\nreturn Main;"
   description=Shadow Pass Switch
   outputType=CMOT_Float2
   inputs [
    {inputName=Main, input {expression=&self:MaterialExpressionAdd_42}}
    {inputName=Shadow, input {expression=&self:MaterialExpressionReroute_30}}
   ]
  }
  n138 = MaterialFunctionCall(
   Texture Object=TextureObjectParameter() {parameterName=BaseColor, texture=&Black}
   R UV=@n137
   G UV=@n96
   B UV=@n105
   Weights=@n106
   Derivative UVs=@n107
  ) {materialFunction=&FrameBlendWeights_TextureObject}
  MaterialExpressionComponentMask_31 = ComponentMask(None=@n138.Result) {a=true}
  MaterialExpressionComponentMask_33 = ComponentMask(None=@n138.R Sample) {a=true}
  n139 = Custom(Main=@MaterialExpressionComponentMask_31, Shadow=@MaterialExpressionComponentMask_33) {
   code="##if SHADOW_DEPTH_SHADER \r\n{\r\nreturn Shadow;\r\n}\r\n##endif\r\nreturn Main;"
   description=Shadow Pass Switch
   inputs [
    {
     inputName=Main
     input {expression=&self:MaterialExpressionComponentMask_31}
    }
    {
     inputName=Shadow
     input {expression=&self:MaterialExpressionComponentMask_33}
    }
   ]
  }
  n140 = MaterialFunctionCall(Default=@n139, Mobile Override=@n110.BaseColor A) {materialFunction=&OverrideMobile}
  n141 = ComponentMask(None=@n140.Result) {r=true}
  n142 = StaticBool()
  NamedRerouteDeclaration(None=StaticSwitch(True=Add(A=@n136, B=@n141), False=@n141, Value=@n142)) {name=OpacityMask Final, nodeColor=<rgba 0.2656 1 0 1>}
  NamedRerouteDeclaration(None=StaticSwitchParameter(True=NamedRerouteUsage(), False=NamedRerouteUsage()) {group=Coloring Systems, parameterName=UseAsteriaColoring}) {name=SubSurfaceColor Final, nodeColor=<rgba 1 0 0.2109 1>}
  n143 = VectorParameter() {parameterName=Emissive Overlay, defaultValue=<rgba 1 1 1 1>}
  n144 = TextureSampleParameter2D(UVs=Multiply(A=TextureCoordinate(), B=ScalarParameter() {defaultValue=1, parameterName=EmissiveScale})) {parameterName=EmissiveOverlay, samplerSource=SSM_Wrap_WorldGroupSettings, texture=&Black}
  NamedRerouteDeclaration(None=StaticSwitchParameter(True=Multiply(A=@n143.RGB, B=@n144.RGB), False=NamedRerouteUsage()) {parameterName=UseSimpleEmissiveOverlay}) {name=Emissive Final, nodeColor=<rgba 1 0.6562 0 1>}
  n145 = SphereMask(A=@n13, B=@n15, Radius=ScalarParameter() {defaultValue=0.03, parameterName=FrameDebugRadius})
  n146 = MaterialFunctionCall(UV=@n17) {materialFunction=&TriangleInterpolator}
  n147 = MaterialFunctionCall(UV=Frac(None=@n38.PS)) {materialFunction=&TriangleInterpolator}
  n148 = Multiply(
   A=LinearInterpolate(A=LinearInterpolate(Alpha=@n146.Mask) {constA=1, constB=0}, B=LinearInterpolate(Alpha=@n146.Mask), Alpha=@n147.Mask)
   B=@n147.Result
  )
  n149 = Add(
   A=Multiply(A=Multiply(A=@n145, B=@n148)) {constB=0.2}
   B=SphereMask(A=TextureCoordinate(), B=@n11, Radius=ScalarParameter() {defaultValue=0.01, parameterName=FrameDebugRadius2})
  )
  n150 = Constant()
  n151 = Multiply(A=Multiply(A=Add(A=@n149, B=@n150)) {constB=2}, B=ScalarParameter() {defaultValue=2, parameterName=DebugBoost})
  NamedRerouteDeclaration(None=@n151) {name=Debug Frames, nodeColor=<rgba 0.4531 1 0 1>}
  NamedRerouteDeclaration(None=StaticSwitch(True=@n135, False=Constant(), Value=@n142)) {name=DebugFrame Emissive, nodeColor=<rgba 0.6875 1 0 1>}
  n152 = MaterialFunctionCall(Default=ComponentMask(None=@n138.Result) {b=true, g=true, r=true}, Mobile Override=@n110.BaseColor RGB) {materialFunction=&OverrideMobile}
  NamedRerouteDeclaration(None=@n152.Result) {name=BaseColorTexture, nodeColor=<rgba 0 0.6484 1 1>}
  n153 = NamedRerouteUsage()
  n154 = ScalarParameter() {defaultValue=0.2, parameterName=Desaturation}
  NamedRerouteDeclaration(None=StaticSwitchParameter(True=Desaturation(None=@n153, Fraction=@n154), False=@n153) {parameterName=UseDesaturation}) {name=BaseColorWithOptionalDesat, nodeColor=<rgba 0.6875 0 1 1>}
  n155 = VectorParameter() {parameterName=BaseColor Multiplier, defaultValue=<rgba 1 1 1 1>}
  n156 = Multiply(A=@n155.RGB, B=@n155.RGB)
  n157 = VectorParameter() {parameterName=BaseColor Multiplier_Highlights, defaultValue=<rgba 1 1 1 1>}
  n158 = Multiply(A=@n157.RGB, B=@n157.RGB)
  n159 = ComponentMask(None=NamedRerouteUsage()) {b=true}
  n160 = ScalarParameter() {defaultValue=0.7, parameterName=ColorMaskMin}
  n161 = Saturate(None=Divide(A=Subtract(A=@n159, B=@n160), B=ScalarParameter() {defaultValue=0.2, parameterName=ColorMaskLen}))
  n162 = Multiply(
   A=StaticSwitchParameter(True=LinearInterpolate(A=@n156, B=@n158, Alpha=@n161), False=@n156) {parameterName=UseTwoColors}
   B=ScalarParameter() {defaultValue=1, parameterName=BaseColorBrightness}
  )
  n163 = LinearInterpolate(A=VectorParameter() {parameterName=BaseColor Multiplier_Trunk, defaultValue=<rgba 1 1 1 1>}.RGB, B=@n162, Alpha=NamedRerouteUsage())
  n164 = NamedRerouteUsage()
  n165 = NamedRerouteUsage()
  NamedRerouteDeclaration(None=StaticSwitchParameter(True=Multiply(A=@n163, B=@n164), False=Multiply(A=@n162, B=@n165)) {parameterName=ExtractTrunkMask}) {name=Legacy BaseColor, nodeColor=<rgba 1 0.375 0 1>}
  n166 = StaticSwitchParameter(True=NamedRerouteUsage(), False=Constant() {r=1}) {parameterName=ExtractTrunkMask}
  n167 = Divide(A=@n24)
  n168 = ComponentMask(None=@n167) {r=true}
  n169 = ScalarParameter() {defaultValue=0.5, group=Color, parameterName=ValueVariation}
  n170 = ScalarParameter() {bUseCustomPrimitiveData=true, defaultValue=0.5, group=Color, parameterName=ValueVariation_primdata, primitiveDataIndex=5}
  n171 = PerInstanceCustomData() {dataIndex=5, desc=Alpha}
  n172 = VertexInterpolator(VS=PerInstanceCustomData(DefaultValue=Constant()) {dataIndex=3, desc=Alpha})
  n173 = VectorParameter() {bUseCustomPrimitiveData=true, parameterName=Color1_Base_primdata}
  n174 = LinearInterpolate(A=@n173.A, B=@n172.PS, Alpha=@n172.PS)
  n175 = StaticSwitchParameter(True=LinearInterpolate(A=@n169, B=LinearInterpolate(A=@n170, B=@n171, Alpha=@n172.PS), Alpha=@n174), False=@n169) {parameterName=UsePrimitiveDataSeasonColoring}
  n176 = ScalarParameter() {defaultValue=-0.03, group=Color, parameterName=Hue Variation}
  n177 = ScalarParameter() {bUseCustomPrimitiveData=true, defaultValue=-0.03, group=Color, parameterName=Hue Variation_primdata, primitiveDataIndex=4}
  n178 = PerInstanceCustomData() {dataIndex=4, desc=Alpha}
  n179 = StaticSwitchParameter(True=LinearInterpolate(A=@n176, B=LinearInterpolate(A=@n177, B=@n178, Alpha=@n172.PS), Alpha=@n174), False=@n176) {parameterName=UsePrimitiveDataSeasonColoring}
  n180 = VectorParameter() {group=Color, parameterName=Color2_Lit, defaultValue=<rgba 0.2016 0.2195 0.0423 1>}
  n181 = LinearInterpolate(A=@n173.RGB, B=VertexInterpolator(VS=PerInstanceCustomData3Vector!(DefaultValue) {desc=Color}).PS, Alpha=@n172.PS)
  n182 = Multiply(A=ScalarParameter() {defaultValue=1.5, group=Color, parameterName=PrimDataColorMultiply_Lit}, B=@n181)
  n183 = VectorParameter() {group=Color, parameterName=Color1_Base, defaultValue=<rgba 0.0887 0.1046 0.0152 1>}
  n184 = StaticSwitchParameter(True=LinearInterpolate(A=@n183.RGB, B=@n181, Alpha=@n174), False=@n183.RGB) {parameterName=UsePrimitiveDataSeasonColoring}
  n185 = LinearInterpolate(
   A=StaticSwitchParameter(True=LinearInterpolate(A=@n180.RGB, B=@n182, Alpha=@n174), False=@n180.RGB) {parameterName=UsePrimitiveDataSeasonColoring}
   B=@n184
   Alpha=@n32
  )
  n186 = VectorParameter() {group=Color, parameterName=Color3_Shadows, defaultValue=<rgba 0.0356 0.0612 0.0116 1>}
  n187 = ScalarParameter() {defaultValue=0.5, group=Color, parameterName=PrimDataColorMultiply_Shadows}
  n188 = StaticSwitchParameter(True=LinearInterpolate(A=@n186.RGB, B=Multiply(A=@n187, B=@n181), Alpha=@n174), False=@n186.RGB) {parameterName=UsePrimitiveDataSeasonColoring}
  n189 = Multiply(
   A=Add(A=Multiply(A=@n168, B=@n175))
   B=MaterialFunctionCall(Hue Shift Percentage=Multiply(A=@n168, B=@n179), Texture=LinearInterpolate(A=@n185, B=@n188, Alpha=@n35)) {materialFunction=&HueShift}.Result
  )
  n190 = ShadingPathSwitch(Default=@n189, Deferred=@n189, Forward=@n184, Mobile=@n184)
  n191 = Add(A=Multiply(A=Multiply(A=@n190, B=@n166)) {constB=2})
  n192 = ComponentMask(None=NamedRerouteUsage()) {b=true, g=true, r=true}
  n193 = StaticSwitchParameter(True=Multiply(A=Subtract(A=@n191, B=@n166), B=@n192), False=LinearInterpolate(A=@n192, B=@n190, Alpha=@n166)) {defaultValue=true, group=Apollo, parameterName=UseDiffuse, sortPriority=1}
  n194 = ScalarParameter() {group=Apollo, parameterName=Fake SSS Scale}
  n195 = Add(A=@n194, B=ScalarParameter() {group=Apollo, parameterName=Fake SSS Scale add Mobile})
  n196 = Multiply(A=@n193, B=Multiply(A=@n166, B=FeatureLevelSwitch(Default=@n194, ES2=@n195, ES3_1=@n195, SM4_REMOVED=@n194, SM5=@n194, SM6=@n194)))
  NamedRerouteDeclaration(None=ComponentMask(None=Add(A=@n193, B=@n196)) {b=true, g=true, r=true}) {name=ApolloColorSystemBaseColor, nodeColor=<rgba 1 0.1875 0 1>}
  n197 = VectorParameter() {parameterName=TrunkExtractColor, defaultValue=<rgba 1 1 1 1>}
  n198 = VectorParameter() {parameterName=ExtractColor, defaultValue=<rgba 0.314 0.1046 0.0395 1>}
  n199 = DotProduct(
   A=Normalize(VectorInput=StaticSwitchParameter(True=@n197.RGB, False=@n198.RGB) {group=Coloring Systems, parameterName=UseAsteriaColoring})
   B=Normalize(VectorInput=NamedRerouteUsage())
  )
  n200 = Divide(
   A=Subtract(A=Abs(None=@n199), B=ScalarParameter() {defaultValue=0.9399, parameterName=TrunkMaskMin})
   B=ScalarParameter() {defaultValue=0, parameterName=TrunkMaskLen}
  )
  NamedRerouteDeclaration(None=OneMinus(None=Saturate(None=@n200))) {name=ExtractTrunkMask, nodeColor=<rgba 1 0 0.4688 1>}
  n201 = VectorParameter() {bUseCustomPrimitiveData=true, parameterName=Color1_Base_primdata}
  n202 = NamedRerouteDeclaration(None=Add(A=@n201.A, B=PerInstanceCustomData() {dataIndex=3})) {name=PrimDataBlendAlpha, nodeColor=<rgba 0 1 0.3203 1>}
  n203 = LinearInterpolate(
   A=MaterialFunctionCall!(
    Float In
    Vector_In=VectorParameter() {group=Color, parameterName=Color1_Base, defaultValue=<rgba 0.0887 0.1046 0.0152 1>}.RGB
    Index=Constant() {r=24}
   ) {materialFunction=/Game/Environments/Asteria/Foliage/Trees/MasterMaterials/MaterialFunctions/MF_AsteriaTree_ArrayAttribute}.Vector_Out
   B=Add(A=PerInstanceCustomData3Vector(), B=@n201.RGB)
   Alpha=@n202
  )
  NamedRerouteDeclaration(None=@n203) {name=AsteriaLeafColor, nodeColor=<rgba 0 1 0.5547 1>}
  n204 = NamedRerouteUsage()
  n205 = VectorParameter() {parameterName=CustomTrunkColorImposterHack, defaultValue=<rgba 1 1 1 0>}
  n206 = ScalarParameter() {defaultValue=1, group=Color, parameterName=Fake_SSS_Color_Mult}
  n207 = Multiply(
   A=NamedRerouteUsage()
   B=MaterialFunctionCall(Hue Shift Percentage=NamedRerouteUsage(), Texture=NamedRerouteUsage()) {materialFunction=&HueShift}.Result
  )
  n208 = Multiply(A=Multiply(A=ScalarParameter() {defaultValue=1, group=Shading, parameterName=Brightness}, B=@n207), B=@n19)
  n209 = NamedRerouteUsage()
  NamedRerouteDeclaration(None=LinearInterpolate(A=Divide(A=@n204, B=@n205.RGB), B=Multiply(A=@n206, B=@n208), Alpha=@n209)) {name=AsteriaColorSystem, nodeColor=<rgba 0.6172 1 0 1>}
  n210 = NamedRerouteUsage()
  n211 = Add(A=@n210, B=ScalarParameter() {group=World Hue Variation, parameterName=World Hue Variation Bias})
  n212 = ScalarParameter() {defaultValue=0.0125, group=World Hue Variation, parameterName=World Hue Variation Strength}
  n213 = Constant() {desc=Fixed value For Wold HueVariation, r=0.0125}
  NamedRerouteDeclaration(None=Multiply(
    A=StaticSwitchParameter(True=@n211, False=@n210) {group=World Hue Variation, parameterName=Enable Global Hue Variation Tweaking}
    B=StaticSwitchParameter(True=@n212, False=@n213) {group=World Hue Variation, parameterName=Enable Global Hue Variation Tweaking}
   )) {name=AsteriaHueShift, nodeColor=<rgba 0 1 0.7891 1>}
  n214 = Constant() {r=0.1}
  n215 = NamedRerouteUsage()
  NamedRerouteDeclaration(None=Add(A=Multiply(A=@n214, B=@n215))) {name=AsteriaValueVariation, nodeColor=<rgba 0 0.0625 1 1>}
  n216 = MaterialFunctionCall(Float3=TransformPosition(None=Constant3Vector()) {transformSourceType=TRANSFORMPOSSOURCE_Instance, transformType=TRANSFORMPOSSOURCE_World}) {materialFunction=&BreakOutFloat3Components}
  n217 = Constant() {r=500}
  n218 = Constant() {r=1100}
  n219 = Add(A=Sine(None=Divide(A=@n216.R, B=@n217)), B=Sine(None=Divide(A=@n216.G, B=@n218)))
  NamedRerouteDeclaration(None=Divide(A=Add(A=@n219) {constB=2}) {constB=4}) {name=PerActorRandom, nodeColor=<rgba 0 1 0.7891 1>}
  n220 = ScalarParameter() {defaultValue=0.03, parameterName=Gcut}
  n221 = Divide(A=Subtract(A=ComponentMask(None=@n152.Result) {g=true}, B=@n220) {constB=0.3}, B=@n220)
  n222 = VectorParameter() {parameterName=SubsurfaceColor, defaultValue=<rgba 1 1 1 5>}
  n223 = Multiply(A=Multiply(A=@n152.Result, B=Saturate(None=@n221)), B=@n222.RGB)
  n224 = ScalarParameter() {defaultValue=0.1, desc=">> Start of inner edge for edge glow", parameterName=Scatter Mask Min}
  n225 = OneMinus(None=@n140.Result)
  n226 = Multiply(
   A=@n222.A
   B=Divide(
    A=Subtract(A=@n225, B=@n224)
    B=ScalarParameter() {defaultValue=2, desc=">> Width of gradient for edge glow", parameterName=Scatter Mask Len}
   )
  )
  n227 = Add(A=@n223, B=Multiply(A=@n223, B=Saturate(None=@n226)))
  NamedRerouteDeclaration(None=@n227) {name=SubSurfaceColorLegacy, nodeColor=<rgba 0 1 0.3672 1>}
  n228 = Multiply(A=@n207, B=ScalarParameter() {defaultValue=0.2, group=Shading, parameterName=SubSurface SSSColor})
  NamedRerouteDeclaration(None=LinearInterpolate(A=Constant3Vector(), B=Multiply(A=@n228, B=@n19), Alpha=@n209)) {name=SubSurfaceColorAsteria, nodeColor=<rgba 0 0.7188 1 1>}
  n229 = Normalize(VectorInput=SkyAtmosphereLightDirection())
  n230 = Multiply(A=Multiply(A=NamedRerouteDeclaration(None=@n229) {name=Camera Vector Replace, nodeColor=<rgba 0 1 0.7656 1>}) {constB=-1}) {constB=500000}
  NamedRerouteDeclaration(None=@n230) {name=CameraPosition Replace, nodeColor=<rgba 1 0.5391 0 1>}
  NamedRerouteUsage()
  Normalize(VectorInput=VectorParameter() {parameterName=LightVector, defaultValue=<rgba -1 0 1 0>}.RGB)
  OneMinus(None=@n168)
  PerInstanceCustomData3Vector() {desc=Color}
  Reroute(None=@n111)
  Reroute()
  Reroute()
  Reroute(None=@n151)
  Reroute(None=@n227)
  Reroute(None=@n230)
  Reroute(None=@n229)
  Round(None=@n1)
  ScalarParameter() {defaultValue=1, parameterName=MipForBleed}
  ScalarParameter() {defaultValue=1, parameterName=EnableParallax}
  ScalarParameter() {defaultValue=0.4951, group=Apollo, parameterName=Shift_Gradient}
  ScalarParameter() {defaultValue=-0.3204, group=Apollo, parameterName=Shift_Gradient_2}
  ScalarParameter() {defaultValue=13000, group=Color Variation WorldSpace, parameterName=Color Variation Distance WS, sortPriority=10}
  ScalarParameter() {defaultValue=11000, group=Color Variation WorldSpace, parameterName=Color Variation Distance WS2, sortPriority=10}
  ScalarParameter() {defaultValue=0.1, parameterName=Opacity}
  ScalarParameter() {defaultValue=2, parameterName=Specular}
  ScalarParameter() {defaultValue=0.5, parameterName=Roughness}
  ScalarParameter() {defaultValue=1, parameterName=MipForBleed}
  ScalarParameter() {defaultValue=1, parameterName=MipForBleed}
  n231 = StaticBool() {value=true}
  n232 = NamedRerouteUsage()
  n233 = NamedRerouteUsage()
  n234 = NamedRerouteUsage()
  n235 = Subtract(A=PixelDepth().R, B=ScalarParameter() {defaultValue=100, parameterName=Opacity_Fade_Offset})
  n236 = ScalarParameter() {defaultValue=20057.9883, parameterName=Opacity_Fade_BlendDist}
  n237 = LinearInterpolate(
   A=ScalarParameter() {defaultValue=1, parameterName=Opacity_Near}
   B=ScalarParameter() {defaultValue=0.75, parameterName=Opacity_Far}
   Alpha=Saturate(None=Divide(A=@n235, B=@n236))
  )
  n238 = Add(A=NamedRerouteUsage(), B=ScalarParameter() {defaultValue=0.1126, parameterName=OutLineBias})
  n239 = OneMinus(None=Saturate(None=Round(None=@n238)))
  n240 = CollectionParameter() {
   collection=&FortniteMaterialParameters
   parameterName=SunlightDirectionalVector
  }
  n241 = Normalize(VectorInput=Add(A=ComponentMask(None=@n240) {b=true, g=true, r=true}, B=CameraVectorWS()) {desc=Compromise for Direction light and all othe rlights})
  n242 = SetMaterialAttributes!(
   MaterialAttributes
   Base Color=LinearInterpolate(A=LinearInterpolate(A=@n232, B=@n233, Alpha=@n234) {constA=1}, Alpha=Multiply(A=@n237, B=@n239)) {constB=0}
   Metallic
   Specular=Constant()
   Roughness=Constant() {r=0.02}
   Emissive Color=NamedRerouteUsage()
   Opacity=LinearInterpolate(A=Constant() {r=1}, B=Constant(), Alpha=NamedRerouteUsage()) {constA=1}
   Opacity Mask=NamedRerouteUsage()
   Normal=LinearInterpolate(A=NamedRerouteUsage(), B=NamedRerouteDeclaration(None=@n241) {name=NormalsFinal, nodeColor=<rgba 0 1 0.4609 1>}, Alpha=NamedRerouteUsage())
   World Position Offset=NamedRerouteUsage()
   Subsurface Color=Constant3Vector()
   Pixel Depth Offset=LinearInterpolate(A=NamedRerouteUsage(), B=Constant(), Alpha=NamedRerouteUsage())
  ) {attributeSetTypes [
    #1
    #2
    #3
    #4
    #5
    #6
    #7
    #8
    #9
    #10
    #11
   ]}
  n243 = GetMaterialAttributes(None=MaterialFunctionCall!(In=@n242) {materialFunction=/LimeRock_Tree/Assets/Treehouse/MaterialFunctions/MF_LimeRock_UnlitShading}.Result) {attributeGetTypes [#1]}
  n244 = Constant() {desc=Snow albedo, r=0.7}
  n245 = DotProduct(A=Normalize(VectorInput=NamedRerouteUsage()), B=Constant3Vector() {constant=<rgba 0 0 1 0>})
  n246 = Constant()
  n247 = Saturate(None=Add(
    A=MaterialFunctionCall(In=@n245, Contrast=@n246) {materialFunction=&CheapContrast}.Result
    B=ScalarParameter() {defaultValue=-0.2, group=Snow, parameterName=SnowNormal_Bias}
   ))
  n248 = SetMaterialAttributes!(
   MaterialAttributes=@n243.MaterialAttributes
   Base Color=StaticSwitchParameter(True=LinearInterpolate(A=@n243.Base Color, B=@n244, Alpha=@n247) {constB=0.7}, False=@n243.Base Color) {group=Snow, parameterName=Use_BRSnow}
  ) {attributeSetTypes [#1]}
  n249 = GetMaterialAttributes(None=@n248) {attributeGetTypes [
    #1
    #7
    #9
    #11
   ]}
  n250 = VertexColor()
  n251 = SetMaterialAttributes!(
   MaterialAttributes=@n249.MaterialAttributes
   Base Color=LinearInterpolate(
    A=TextureSampleParameter2D() {parameterName=BaseColor, samplerSource=SSM_Wrap_WorldGroupSettings, texture=&Black}.RGB
    B=@n249.Base Color
    Alpha=@n250.R
   )
   Opacity Mask=LinearInterpolate(A=Constant() {r=1}, B=@n249.Opacity Mask, Alpha=@n250.R)
   World Position Offset=LinearInterpolate(A=Constant3Vector(), B=@n249.World Position Offset, Alpha=@n250.R)
   Pixel Depth Offset=LinearInterpolate(A=Constant(), B=@n249.Pixel Depth Offset, Alpha=@n250.R)
  ) {attributeSetTypes [
    #1
    #7
    #9
    #11
   ]}
  n252 = StaticSwitchParameter(True=@n251, False=@n248) {parameterName=Use_StaticTrunk}
  n253 = StaticSwitchParameter(
   True=MaterialFunctionCall!(
    IsStaticSnow
    UseAlternateStaticSnowMask
    AlternateStaticSnowMask
    "UseRoad&&OcclusionMask"=StaticBool()
    IsForGrassMeshToPushDown
    AdditionalMaskToRemoveSnow
    IsFallingSnowFX
    UseMinimapCoords
    IncreaseSnowUsingSlope
    UseSmoothBorder=@n231
    LerpHighM=Constant() {r=1.5}
    TemporalHeightLerpContrast
    IsLobbyIsland
    UseBR-IslandMask
    IsTree=@n231
    IsForRenderTarget
    UseSparkle
    UseSnowNormals
    UseAdditionalMaskToRemoveSnow
    IsTangentSpaceInput
    UseVertexNormalsForBasicSnow
    UseTangentSnowNormals
    Material=@n252
   ) {materialFunction=/Game/Athena/Apollo/Environments/Landscape/Materials/Biome/UniqueBiomeMaterialFunctions/MF_Apollo_Snow_Progression}.Material
   False=@n252
  ) {group=Snow, parameterName=Use_BRSnow-Progressive}
  StaticSwitchParameter(
   True=MaterialFunctionCall!(In=@n253, Mask=Constant() {r=1}, IsImpostor=StaticBool() {value=true}) {materialFunction=/Game/Environments/Apollo/Foliage/MaterialFunctions/MF_TreeFresnel}.Result
   False=@n253
  ) {group=Coloring Systems, parameterName=UseApolloColoring, sortPriority=-100}
  n254 = VertexColor()
  StaticSwitchParameter(
   True=LinearInterpolate(
    A=TextureSampleParameter2D() {parameterName=BaseColor, samplerSource=SSM_Wrap_WorldGroupSettings, texture=&Black}.RGB
    B=NamedRerouteUsage()
    Alpha=@n254.R
   )
   False=NamedRerouteUsage()
  ) {parameterName=Use_StaticTrunk}
  n255 = NamedRerouteUsage()
  StaticSwitchParameter(True=LinearInterpolate(A=Constant() {r=1}, B=@n255, Alpha=@n254.R), False=@n255) {parameterName=Use_StaticTrunk}
  n256 = NamedRerouteUsage()
  n257 = VertexColor()
  StaticSwitchParameter(True=LinearInterpolate(A=Constant3Vector(), B=@n256, Alpha=@n257.R), False=@n256) {parameterName=Use_StaticTrunk}
  n258 = NamedRerouteUsage()
  StaticSwitchParameter(True=LinearInterpolate(A=Constant(), B=@n258, Alpha=@n257.R), False=@n258) {parameterName=Use_StaticTrunk}
  n259 = Constant2Vector()
  n260 = VectorParameter() {parameterName=ReferencePoint, defaultValue=<rgba 65995.8828 46048.0703 0 1>}
  n261 = Multiply(A=Subtract(A=WorldPosition().XYZ, B=@n260.RGB), B=Constant3Vector() {constant=<rgba 1 1 0 0>})
  n262 = Subtract(A=VectorParameter() {parameterName=ReferencePoint_1, defaultValue=<rgba 88680.0391 35843.0195 0 1>}.RGB, B=@n260.RGB)
  n263 = Normalize(VectorInput=@n262)
  n264 = Constant() {r=-5000}
  n265 = If(
   A=Divide(
    A=DotProduct(A=@n261, B=@n263)
    B=MaterialFunctionCall(Vector 2=@n262) {materialFunction=&VectorLength}.V2 Length
   )
   B=Constant() {desc=Was originally CubePathThreshold}
   "A > B"=Constant()
   "A == B"=@n264
   "A < B"=@n264
  )
  n266 = NamedRerouteUsage()
  StaticSwitchParameter(True=Add(A=@n266, B=AppendVector(A=@n259, B=@n265)), False=@n266) {parameterName=CubeDestroyPath}
  StaticSwitchParameter(
   True=NamedRerouteUsage()
   False=StaticSwitchParameter(True=NamedRerouteUsage(), False=NamedRerouteUsage()) {group=Coloring Systems, parameterName=UseApolloColoring, sortPriority=-100}
  ) {group=Coloring Systems, parameterName=UseAsteriaColoring}
  n267 = VectorParameter() {group=Color VertexColor Tints, parameterName=Green Vertex Channel Color, sortPriority=4, defaultValue=<rgba 1 1 1 1>}
  n268 = VectorParameter() {bUseCustomPrimitiveData=true, desc=20 not used, parameterName=Color2_Base_primdata, primitiveDataIndex=17, defaultValue=<rgba 0 0 0 1>}
  n269 = Add(A=@n268.RGB, B=PerInstanceCustomData3Vector() {dataIndex=17})
  n270 = LinearInterpolate(
   A=ScalarParameter() {group=Color VertexColor Tints, parameterName=Green Vertex Channel Mask Value LOW, sortPriority=5}
   B=ScalarParameter() {group=Color VertexColor Tints, parameterName=Green Vertex Channel Mask Value HIGH, sortPriority=5}
   Alpha=VertexColor().G
  )
  StaticSwitch(
   True=Constant3Vector() {constant=<rgba 1 1 1 0>}
   False=LinearInterpolate(A=@n203, B=LinearInterpolate(A=@n267.RGB, B=@n269, Alpha=@n202), Alpha=Saturate(None=@n270))
   Value=NamedRerouteUsage()
  ) {desc=For ImpostorBakeing}
  TextureCoordinate() {uTiling=10, vTiling=10}
  VectorParameter() {parameterName=LightVector, defaultValue=<rgba -1 0 1 0>}
  VertexInterpolator()
  VertexInterpolator(VS=@n25)
  VertexInterpolator(VS=@n167)
 ]
}
