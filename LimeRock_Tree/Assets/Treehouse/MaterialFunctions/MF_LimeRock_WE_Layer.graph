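# Material layer function for LimeRock tree assets, outputting material attributes with zero metallic and specular and an optional normal map.
# Base color takes a diffuse texture or color, optionally multiplied by a world-aligned or UV-mapped overlay.
# A roughness-based cell-shading value uses either a narrow outline-exclusion range or vertex red, optionally randomized per object from world position.
# Its lowest range is noted as a foliage mask.
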MaterialFunction /LimeRock_Tree/Assets/Treehouse/MaterialFunctions/MF_LimeRock_WE_Layer
graph {nodes [
  n1 = Constant()
  n2 = FunctionInput() {inputName=NormalTexture, inputType=FunctionInput_Texture2D, sortPriority=40}
  n3 = SetMaterialAttributes!(
   MaterialAttributes
   BaseColor=NamedRerouteUsage()
   Metallic=@n1
   Specular=@n1
   Roughness=NamedRerouteUsage()
   Normal=StaticSwitch(
    True=TextureSample(Tex=@n2) {samplerType=SAMPLERTYPE_Normal}.RGB
    False=Constant3Vector() {constant=<rgba 0 0 1 0>}
    Value=FunctionInput() {inputName=UseNormal, inputType=FunctionInput_StaticBool, sortPriority=200}
   )
  ) {attributeSetTypes [
    #1
    #2
    #3
    #4
    #5
   ]}
  FunctionOutput(None=@n3)
  n4 = StaticSwitch(
   True=TextureSample(Tex=FunctionInput() {inputName=DiffuseTexture, inputType=FunctionInput_Texture2D, sortPriority=20}) {texture=/Game/Athena/WinterArt/Textures/T_White_Dif}.RGB
   False=FunctionInput() {inputName=DiffuseColor, sortPriority=10}
   Value=FunctionInput() {inputName=UseDiffuse, inputType=FunctionInput_StaticBool, sortPriority=100}
  )
  n5 = FunctionInput() {inputName=Overlay Scale, sortPriority=302}
  n6 = FunctionInput() {inputName=Overlay_Projection_Contrast, inputType=FunctionInput_Scalar, sortPriority=103, previewValue=<1 0 0 1>}
  n7 = FunctionInput() {inputName=OverlayTexture, inputType=FunctionInput_Texture2D, sortPriority=30}
  n8 = MaterialFunctionCall(
   TextureObject=@n7
   TextureSize=Multiply(A=FunctionInput() {inputName=Overlay WS Size, inputType=FunctionInput_Scalar, sortPriority=301}, B=@n5)
   ProjectionTransitionContrast=@n6
  ) {materialFunction=&WorldAlignedTexture}
  n9 = StaticBool()
  n10 = TextureCoordinate()
  n11 = ComponentMask(None=@n5) {g=true, r=true}
  n12 = StaticSwitch(
   True=StaticSwitch(
    True=@n8.XY Texture
    False=@n8.XY Texture
    Value=FunctionInput(Preview=@n9) {bUsePreviewValueAsDefault=true, inputName=Overlay_XYProjectionOnly, inputType=FunctionInput_StaticBool, sortPriority=102}
   )
   False=TextureSample(UVs=Divide(A=@n10, B=@n11), Tex=@n7) {samplerType=SAMPLERTYPE_Masks, texture=/Game/Global/Textures/Default/Blanks/T_White}.R
   Value=FunctionInput() {description=Overylay use WS, inputName="Static Bool Param (True)", inputType=FunctionInput_StaticBool, sortPriority=300}
  )
  NamedRerouteDeclaration(None=StaticSwitch(True=Multiply(A=@n4, B=@n12), False=@n4, Value=FunctionInput() {inputName=UseOverlay, inputType=FunctionInput_StaticBool, sortPriority=101})) {name=BaseColor, nodeColor=<rgba 0 1 0.9766 1>}
  n13 = Constant() {r=0.05}
  n14 = Constant() {r=0.08}
  n15 = MaterialFunctionCall(Float3=TransformPosition(None=Constant3Vector()) {transformType=TRANSFORMPOSSOURCE_World}) {materialFunction=&BreakOutFloat3Components}
  n16 = Add(A=@n15.R, B=@n15.R)
  n17 = Frac(None=Divide(A=Add(A=@n16, B=@n15.B)) {constB=100})
  n18 = VertexColor()
  n19 = LinearInterpolate(Alpha=Multiply(A=@n17, B=LinearInterpolate(Alpha=@n18.R) {constA=0.1, constB=0.9})) {constA=0.1}
  n20 = LinearInterpolate(Alpha=@n18.R) {constA=0.1, desc=Range 0 - 0.1 is used to maks Foliage}
  NamedRerouteDeclaration(None=StaticSwitchParameter(
    True=LinearInterpolate(A=@n13, B=@n14, Alpha=@n17)
    False=StaticSwitchParameter(True=@n19, False=@n20) {group=CellShading, parameterName=UseRandomCellShade}
   ) {group=CellShading, parameterName=UseOutlineExlusionMask}) {name=RoughnessCellShading, nodeColor=<rgba 1 0.6563 0 1>}
 ]}
