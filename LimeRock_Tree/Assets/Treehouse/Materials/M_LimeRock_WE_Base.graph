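# Nanite-ready base material for LimeRock Treehouse environment meshes, with up to four tintable layers, each with optional diffuse, normal and projected overlay textures.
# Static switches set the layer count, blending layers through hard step masks from UV1 or UV2's U coordinate.
# The result passes through normal transform, color quantization, layer-masked grass blend, unlit shading and destructible damage functions.
# A normal-quantizing branch and per-layer flat-shading switches are left unconnected.
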Material /LimeRock_Tree/Assets/Treehouse/Materials/M_LimeRock_WE_Base
graph {
 properties {
  bAutomaticallySetUsageInEditor=false
  bTangentSpaceNormal=false
  bUseMaterialAttributes=true
  bUsedWithInstancedStaticMeshes=true
  bUsedWithNanite=true
  shadingModel=MSM_FromMaterialExpression
  usedShadingModels=Default Lit
 }
 nodes [
  Composite() {subgraphName=UV Mask}
  n1 = ComponentMask(None=TextureCoordinate() {coordinateIndex=2}) {r=true}
  n2 = ComponentMask(None=TextureCoordinate() {coordinateIndex=1}) {r=true}
  n3 = Constant()
  n4 = Constant() {r=1}
  n5 = If(
   A=Constant() {r=1}
   B=StaticSwitchParameter(True=@n1, False=@n2) {group=2_Layer, parameterName=Use UV2 For Layer 2 Mask, sortPriority=-2}
   "A > B"=@n3
   "A < B"=@n4
  ) {desc=0}
  MaterialExpressionReroute_72 = Reroute(None=@n5)
  n6 = BlendMaterialAttributes(A=NamedRerouteUsage(), B=NamedRerouteUsage(), Alpha=@MaterialExpressionReroute_72.Output)
  n7 = If(
   A=Constant() {r=2}
   B=StaticSwitchParameter(True=@n1, False=@n2) {group=3_Layer, parameterName=Use UV2 For Layer 3 Mask, sortPriority=-2}
   "A > B"=@n3
   "A < B"=@n4
  )
  MaterialExpressionReroute_73 = Reroute(None=@n7)
  n8 = BlendMaterialAttributes(A=@n6, B=NamedRerouteUsage(), Alpha=@MaterialExpressionReroute_73.Output2)
  n9 = NamedRerouteUsage()
  n10 = If(
   A=Constant() {r=3}
   B=StaticSwitchParameter(True=@n1, False=@n2) {group=4_Layer, parameterName=Use UV2 For Layer 4 Mask, sortPriority=-2}
   "A > B"=@n3
   "A < B"=@n4
  )
  MaterialExpressionReroute_74 = Reroute(None=@n10)
  n11 = StaticSwitch(True=@n6, False=NamedRerouteUsage(), Value=StaticBoolParameter() {group=2_Layer, parameterName=Use 2 Layers, sortPriority=-10})
  n12 = StaticBoolParameter() {group=3_Layer, parameterName=Use 3 Layers, sortPriority=-10}
  n13 = GetMaterialAttributes(None=StaticSwitch(
    True=BlendMaterialAttributes(A=@n8, B=@n9, Alpha=@MaterialExpressionReroute_74.Output3)
    False=StaticSwitch(True=@n8, False=@n11, Value=@n12)
    Value=StaticBoolParameter() {group=4_Layer, parameterName=Use 4 Layers, sortPriority=-10}
   )) {attributeGetTypes [#1]}
  n14 = MaterialFunctionCall!(In=SetMaterialAttributes!(MaterialAttributes=@n13.MaterialAttributes, Normal=Transform(None=@n13.Normal)) {attributeSetTypes [#1]}) {materialFunction=/LimeRock_Tree/Assets/Treehouse/MaterialFunctions/M_LimeRock_WE_QantizeColor}
  n15 = StaticBool() {value=true}
  n16 = NamedRerouteUsage()
  n17 = NamedRerouteUsage()
  n18 = NamedRerouteUsage()
  MaterialFunctionCall(
   Material_In=MaterialFunctionCall!(In=MaterialFunctionCall!(In=@n14.Result, IsWorldSpace=@n15, L2_Mask=@n16, L3_Mask=@n17, L4_Mask=@n18, UseLayerMasking) {materialFunction=/LimeRock_Tree/Assets/Treehouse/Materials/Grass/M_LimeRock_GrassBlend}.Result) {materialFunction=/LimeRock_Tree/Assets/Treehouse/MaterialFunctions/MF_LimeRock_UnlitShading}.Result
   UseCrackedDamageTextures=StaticBool()
   UseAdditiveWPO=StaticBool() {value=true}
  ) {materialFunction=&MF_FN_DestructibleDamage}
  NamedRerouteDeclaration(None=@MaterialExpressionReroute_72.Output) {name=L2_Mask, nodeColor=<rgba 1 0 0.2344 1>}
  NamedRerouteDeclaration(None=@MaterialExpressionReroute_73.Output2) {name=L3_Mask, nodeColor=<rgba 1 0 0.2344 1>}
  n19 = TextureObjectParameter() {
   desc=Not Really Specular related
   group=1_Layer_Overlay
   parameterName=SpecularMasks
   samplerType=SAMPLERTYPE_Masks
   sortPriority=1000
   texture=/Game/Global/Textures/Default/Blanks/T_White
   channelNames {a="", b="", g="", r=OverlayTexture}
  }
  n20 = TextureObjectParameter() {group=1_Layer, parameterName=Normals, samplerType=SAMPLERTYPE_Normal, sortPriority=0, texture=/Engine/EngineMaterials/FlatNormal}
  n21 = VectorParameter() {group=1_Layer_Overlay, parameterName=L1_OverlayScale, sortPriority=58, defaultValue=<rgba 1 1 1 1>}
  NamedRerouteDeclaration(None=MaterialFunctionCall!(
    DiffuseColor=VectorParameter() {group=1_Layer, parameterName="L1 Tint (Expand)", defaultValue=<rgba 0.6105 0.3613 0.2159 1>}.RGB
    DiffuseTexture=TextureObjectParameter() {group=1_Layer, parameterName=Diffuse, sortPriority=0, texture=/Game/Global/Textures/Default/Blanks/T_MidgreysRGB}
    OverlayTexture=@n19
    NormalTexture=@n20
    UseDiffuse=StaticBoolParameter() {group=1_Layer, parameterName=L1 Use Diffuse}
    UseOverlay=StaticBoolParameter() {group=1_Layer_Overlay, parameterName=L1 Use OverlayTexture, sortPriority=3}
    Overlay_XYProjectionOnly=StaticBoolParameter() {defaultValue=true, group=1_Layer_Overlay, parameterName=L1 Overlay XY Only, sortPriority=5}
    Overlay_Projection_Contrast=ScalarParameter() {defaultValue=1, group=1_Layer_Overlay, parameterName=L1_Overlay Porjection Contrast}
    UseNormal=StaticBoolParameter() {group=1_Layer, parameterName=L1 Use Normal}
    "Static Bool Param (True)"=StaticBoolParameter() {defaultValue=true, group=1_Layer_Overlay, parameterName=L1 Overlay WS projection, sortPriority=5}
    Overlay WS Size=ScalarParameter() {defaultValue=256, group=1_Layer_Overlay, parameterName=L1_OverlaySize}
    Overlay Scale=@n21.RGB
   ) {materialFunction=/LimeRock_Tree/Assets/Treehouse/MaterialFunctions/MF_LimeRock_WE_Layer}.Result) {name=L1, nodeColor=<rgba 0 1 0.3438 1>}
  n22 = TextureObjectParameter() {group=2_Layer, parameterName=Diffuse_Texture_2, sortPriority=0, texture=&T_MidgreysRGB}
  n23 = TextureObjectParameter() {
   desc="R = Specular\r\nG = Metal\r\nB = Roughness"
   group=2_Layer_Overlay
   parameterName=SpecularMasks_2
   samplerType=SAMPLERTYPE_Masks
   sortPriority=0
   texture=&T_White
  }
  n24 = TextureObjectParameter() {group=2_Layer, parameterName=Normals_Texture_2, samplerType=SAMPLERTYPE_Normal, sortPriority=0, texture=&FlatNormal}
  n25 = VectorParameter() {group=2_Layer_Overlay, parameterName=L2_OverlayScale, sortPriority=58, defaultValue=<rgba 1 1 1 1>}
  NamedRerouteDeclaration(None=MaterialFunctionCall(
    DiffuseColor=VectorParameter() {group=2_Layer, parameterName="L2 Tint (Expand)", defaultValue=<rgba 0.6105 0.3613 0.2159 1>}.RGB
    DiffuseTexture=@n22
    OverlayTexture=@n23
    NormalTexture=@n24
    UseDiffuse=StaticBoolParameter() {group=2_Layer, parameterName=L2 Use Diffuse}
    UseOverlay=StaticBoolParameter() {group=2_Layer_Overlay, parameterName=L2 Use OverlayTexture, sortPriority=3}
    Overlay_XYProjectionOnly=StaticBoolParameter() {defaultValue=true, group=2_Layer_Overlay, parameterName=L2 Overlay XY Only, sortPriority=5}
    Overlay_Projection_Contrast=ScalarParameter() {defaultValue=1, group=2_Layer_Overlay, parameterName=L2_Overlay Porjection Contrast}
    UseNormal=StaticBoolParameter() {group=2_Layer, parameterName=L2 Use Normal}
    "Static Bool Param (True)"=StaticBoolParameter() {defaultValue=true, group=2_Layer_Overlay, parameterName=L2Overlay WS projection, sortPriority=5}
    Overlay WS Size=ScalarParameter() {defaultValue=256, group=2_Layer_Overlay, parameterName=L2_OverlaySize}
    Overlay Scale=@n25.RGB
   ) {materialFunction=&MF_LimeRock_WE_Layer}.Result) {name=L2, nodeColor=<rgba 0 1 0.3438 1>}
  n26 = TextureObjectParameter() {group=3_Layer, parameterName=Diffuse_Texture_3, sortPriority=0, texture=&T_MidgreysRGB}
  n27 = TextureObjectParameter() {
   desc="R = Specular\r\nG = Metal\r\nB = Roughness"
   group=3_Layer_Overlay
   parameterName=SpecularMasks_3
   samplerType=SAMPLERTYPE_Masks
   sortPriority=0
   texture=&T_White
  }
  n28 = TextureObjectParameter() {group=3_Layer, parameterName=Normals_Texture_3, samplerType=SAMPLERTYPE_Normal, sortPriority=0, texture=&FlatNormal}
  n29 = VectorParameter() {group=3_Layer_Overlay, parameterName=L3_OverlayScale, sortPriority=58, defaultValue=<rgba 1 1 1 1>}
  NamedRerouteDeclaration(None=MaterialFunctionCall(
    DiffuseColor=VectorParameter() {group=3_Layer, parameterName="L3 Tint (Expand)", defaultValue=<rgba 0.6105 0.3613 0.2159 1>}.RGB
    DiffuseTexture=@n26
    OverlayTexture=@n27
    NormalTexture=@n28
    UseDiffuse=StaticBoolParameter() {group=3_Layer, parameterName=L3 Use Diffuse}
    UseOverlay=StaticBoolParameter() {group=3_Layer_Overlay, parameterName=L3 Use OverlayTexture, sortPriority=3}
    Overlay_XYProjectionOnly=StaticBoolParameter() {defaultValue=true, group=3_Layer_Overlay, parameterName=L3 Overlay XY Only, sortPriority=5}
    Overlay_Projection_Contrast=ScalarParameter() {defaultValue=1, group=3_Layer_Overlay, parameterName=L3_Overlay Porjection Contrast}
    UseNormal=StaticBoolParameter() {group=3_Layer, parameterName=L3 Use Normal}
    "Static Bool Param (True)"=StaticBoolParameter() {defaultValue=true, group=3_Layer_Overlay, parameterName=L3 Overlay WS projection, sortPriority=5}
    Overlay WS Size=ScalarParameter() {defaultValue=256, group=3_Layer_Overlay, parameterName=L3_OverlaySize}
    Overlay Scale=@n29.RGB
   ) {materialFunction=&MF_LimeRock_WE_Layer}.Result) {name=L3, nodeColor=<rgba 0 1 0.3438 1>}
  n30 = TextureObjectParameter() {group=4_Layer, parameterName=Diffuse_Texture_4, sortPriority=0, texture=&T_MidgreysRGB}
  n31 = TextureObjectParameter() {
   desc="R = Specular\r\nG = Metal\r\nB = Roughness"
   group=4_Layer_Overlay
   parameterName=SpecularMasks_4
   samplerType=SAMPLERTYPE_Masks
   sortPriority=0
   texture=&T_White
  }
  n32 = TextureObjectParameter() {group=4_Layer, parameterName=Normals_Texture_4, samplerType=SAMPLERTYPE_Normal, sortPriority=0, texture=&FlatNormal}
  n33 = VectorParameter() {group=4_Layer_Overlay, parameterName=L4_OverlayScale, sortPriority=58, defaultValue=<rgba 1 1 1 1>}
  NamedRerouteDeclaration(None=MaterialFunctionCall(
    DiffuseColor=VectorParameter() {group=4_Layer, parameterName="L4 Tint (Expand)", defaultValue=<rgba 0.6105 0.3613 0.2159 1>}.RGB
    DiffuseTexture=@n30
    OverlayTexture=@n31
    NormalTexture=@n32
    UseDiffuse=StaticBoolParameter() {group=4_Layer, parameterName=L4 Use Diffuse}
    UseOverlay=StaticBoolParameter() {group=4_Layer_Overlay, parameterName=L4 Use OverlayTexture, sortPriority=3}
    Overlay_XYProjectionOnly=StaticBoolParameter() {defaultValue=true, group=4_Layer_Overlay, parameterName=L4 Overlay XY Only, sortPriority=5}
    Overlay_Projection_Contrast=ScalarParameter() {defaultValue=1, group=4_Layer_Overlay, parameterName=L4_Overlay Porjection Contrast}
    UseNormal=StaticBoolParameter() {group=4_Layer, parameterName=L4 Use Normal}
    "Static Bool Param (True)"=StaticBoolParameter() {defaultValue=true, group=4_Layer_Overlay, parameterName=L4 Overlay WS projection, sortPriority=5}
    Overlay WS Size=ScalarParameter() {defaultValue=256, group=4_Layer_Overlay, parameterName=L4_OverlaySize}
    Overlay Scale=@n33.RGB
   ) {materialFunction=&MF_LimeRock_WE_Layer}.Result) {name=L4, nodeColor=<rgba 0 1 0.3438 1>}
  NamedRerouteDeclaration(None=@MaterialExpressionReroute_74.Output3) {name=L4_Mask, nodeColor=<rgba 1 0 0.2344 1>}
  PinBase()
  PinBase!(Output=@n5, Output2=@n7, Output3=@n10) {reroutePins [
    Output {expression=&self:MaterialExpressionReroute_72}
    Output2 {expression=&self:MaterialExpressionReroute_73}
    Output3 {expression=&self:MaterialExpressionReroute_74}
   ]}
  n34 = GetMaterialAttributes(None=@n14.Result) {attributeGetTypes [#1]}
  n35 = ScalarParameter() {defaultValue=8, parameterName=QuantizeNormals}
  n36 = Multiply(A=@n34.Normal, B=@n35)
  SetMaterialAttributes(MaterialAttributes=@n34.MaterialAttributes, Normal=Divide(A=Round(None=@n36), B=@n35)) {attributeSetTypes [#1]}
  StaticBoolParameter() {group=FlatShading, parameterName=UseFlatShading_Layer1, sortPriority=1}
  StaticBoolParameter() {group=FlatShading, parameterName=UseFlatShading_Layer2, sortPriority=1}
  StaticBoolParameter() {group=FlatShading, parameterName=UseFlatShading_Layer3, sortPriority=1}
  StaticBoolParameter() {group=FlatShading, parameterName=UseFlatShading_Layer4, sortPriority=1}
 ]
}
