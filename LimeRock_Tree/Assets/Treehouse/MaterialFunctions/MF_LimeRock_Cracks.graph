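# Material function that darkens a material's base color with cracks driven by gameplay destructible damage.
# A world-aligned damage crack texture's red channel plus the damage amount, floored and saturated, forms a hard reveal mask.
# That mask blends from white to the green channel, which multiplies base color in the incoming material attributes.
# Damage defaults to custom primitive data index 13; two test and subtract nodes are unconnected.
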MaterialFunction /LimeRock_Tree/Assets/Treehouse/MaterialFunctions/MF_LimeRock_Cracks
graph {nodes [
  n1 = GetMaterialAttributes(None=FunctionInput() {inputType=FunctionInput_MaterialAttributes}) {attributeGetTypes [#1]}
  n2 = MaterialFunctionCall(
   TextureObject=TextureObject() {texture=&/WildEstate_DamageCracks}
   TextureSize=ScalarParameter() {defaultValue=384, parameterName=Destructible Damage Cracks Size}
  ) {materialFunction=&WorldAlignedTexture}
  n3 = ScalarParameter() {
   bUseCustomPrimitiveData=true
   desc="Destructible Damage. This parameter is controlled by gameplay code.\r\nDefault=Custom Primitive Data Index 13"
   group=z_GameplayParameters_DestructibleDamage
   parameterName=Code_DestructionAmount
   primitiveDataIndex=13
   sortPriority=105
  }
  n4 = FunctionInput(Preview=@n3) {
   bUsePreviewValueAsDefault=true
   description=Default parameter is controlled by code. Custom Primitive Data Index 13
   inputName=DestructibleDamageAmount
   inputType=FunctionInput_Scalar
   sortPriority=200
   previewValue=<0 1 0 1>
  }
  n5 = Add(A=ComponentMask(None=@n2.XYZ Texture) {r=true}, B=@n4)
  n6 = LinearInterpolate(A=Constant() {r=1}, B=ComponentMask(None=@n2.XYZ Texture) {g=true}, Alpha=Saturate(None=Floor(None=@n5)))
  FunctionOutput(None=SetMaterialAttributes!(MaterialAttributes=@n1.MaterialAttributes, BaseColor=Multiply(A=@n1.BaseColor, B=@n6)) {attributeSetTypes [#1]})
  ScalarParameter() {defaultValue=1, parameterName=Test Cracks}
  Subtract(A=@n5) {constB=0.05}
 ]}
