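# Material function for the LimeRock treehouse tree's trunk shading, applied when the static switch IsTrunk is on and otherwise passing input attributes through.
# Builds base color from Trunk_Color, optionally multiplied by a diffuse texture, times the overlay's red channel, with full roughness, zero specular and a switchable, transformed normal.
# These attributes pass through QantizeColor, then GrassBlend in world space without layer masking.
# Also declares two "visual warning" reroutes: a constant color and a world-height sine.
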MaterialFunction /LimeRock_Tree/Assets/Treehouse/MaterialFunctions/MF_LimeRock_WE_Trunk
graph {nodes [
  n1 = VectorParameter() {group=Trunk, parameterName=Trunk_Color, defaultValue=<rgba 1 1 1 1>}
  n2 = Multiply(
   A=@n1.RGB
   B=TextureSampleParameter2D() {group=Trunk, parameterName=Trunk_Diffuse, texture=/Game/Global/Textures/Default/Blanks/T_MidgreysRGB}.RGB
  )
  n3 = TextureSampleParameter2D() {
   group=Trunk
   parameterName=OverlayTexture
   samplerType=SAMPLERTYPE_Masks
   texture=/LimeRock_Tree/Assets/Treehouse/Textures/T_WildEstate_Tree_Bark2_D
  }
  n4 = TextureObjectParameter() {group=Trunk, parameterName=Normals, samplerType=SAMPLERTYPE_Normal, sortPriority=0, texture=/Engine/EngineMaterials/FlatNormal}
  n5 = Constant3Vector() {constant=<rgba 0 0 1 0>}
  n6 = SetMaterialAttributes!(
   MaterialAttributes
   BaseColor=Multiply(A=StaticSwitchParameter(True=@n2, False=@n1.RGB) {group=Trunk, parameterName=UseDiffuseTexture}, B=@n3.R)
   Specular=Constant()
   Roughness=Constant() {r=1}
   Anisotropy
   EmissiveColor
   Opacity
   OpacityMask
   Normal=Transform(None=StaticSwitchParameter(True=@n4, False=@n5) {group=Trunk, parameterName=UseNoramlTexture})
   SubsurfaceColor=NamedRerouteUsage()
   WorldPositionOffset
  ) {attributeSetTypes [
    #1
    #2
    #3
    #4
    #5
    #6
    #7
    #8
    #9
    #10
   ]}
  n7 = MaterialFunctionCall!(In=@n6) {materialFunction=&/M_LimeRock_WE_QantizeColor}
  n8 = StaticBool() {value=true}
  n9 = StaticBool()
  FunctionOutput(None=StaticSwitchParameter(
    True=MaterialFunctionCall!(In=@n7.Result, IsWorldSpace=@n8, L2_Mask, L3_Mask, L4_Mask, UseLayerMasking=@n9) {materialFunction=/LimeRock_Tree/Assets/Treehouse/Materials/Grass/M_LimeRock_GrassBlend}.Result
    False=FunctionInput() {inputType=FunctionInput_MaterialAttributes}
   ) {group=Trunk, parameterName=IsTrunk})
  NamedRerouteDeclaration(None=Constant3Vector() {constant=<rgba 8 0.95 0 0>}) {name=Visual Warning 2, nodeColor=<rgba 1 0.6678 0 1>}
  n10 = ComponentMask(None=WorldPosition().XYZ) {b=true}
  NamedRerouteDeclaration(None=Sine(None=Divide(A=@n10) {constB=50})) {name=VisualWaring 1, nodeColor=<rgba 1 0.7456 0 1>}
 ]}
