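# Base material for stylized LimeRock Wild Estate treehouse trees, with camera-billboarded leaves, random hue shift and an optional hedge look.
# Vertex offsets flatten or rotate leaves around pivots, while color uses quantized sunlight shading from merged leaf clusters or a camera-aligned virtual-plane height map.
# Attributes chain through trunk, switchable wind animation, unlit shading, destructible damage and crack functions.
# Switches handle impostor baking and distance WPO fade, with a Nanite-specific path.
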Material /LimeRock_Tree/Assets/Treehouse/Materials/M_LimeRock_WE_Tree_BaseMaterial
graph {
 properties {
  bAutomaticallySetUsageInEditor=false
  bHasPixelAnimation=true
  bTangentSpaceNormal=false
  bUseMaterialAttributes=true
  bUsedWithInstancedStaticMeshes=true
  bUsedWithNanite=true
  maxWorldPositionOffsetDisplacement=500
  shadingModel=MSM_Subsurface
  usedShadingModels=Subsurface
 }
 nodes [
  Add(A=WorldPosition() {worldPositionShaderOffset=WPT_ExcludeAllShaderOffsets}.XYZ, B=NamedRerouteUsage())
  CameraVectorWS()
  n1 = WorldPosition() {worldPositionShaderOffset=WPT_CameraRelativeNoOffsets}
  n2 = StaticSwitchParameter(True=Add(A=@n1.XYZ, B=NamedRerouteUsage()), False=@n1.XYZ) {group=Billboard, parameterName=UseFallbackBillboard}
  n3 = NamedRerouteUsage()
  n4 = ScalarParameter() {defaultValue=150, group=Color, parameterName=LeafVisualMergeDistance}
  n5 = Round(None=Divide(A=NamedRerouteUsage(), B=@n4))
  n6 = TransformPosition(None=NamedRerouteDeclaration(None=Multiply(A=@n5, B=@n4)) {name=LeafPivot_LocalSpace_Merged, nodeColor=<rgba 0 1 0.5547 1>}) {transformSourceType=TRANSFORMPOSSOURCE_Instance, transformType=TRANSFORMPOSSOURCE_TranslatedWorld}
  n7 = Normalize(VectorInput=@n6)
  n8 = DotProduct(A=Subtract(A=Add(A=@n2, B=@n3), B=@n6), B=@n7)
  n9 = Max(
   A=Subtract(A=Multiply(A=@n8) {constB=-1}, B=ScalarParameter() {defaultValue=66.7177, group=BillBoard WPO, parameterName=BillboardSqueze_Clamp})
   B=Constant()
  )
  Ceil(None=@n9)
  n10 = Transform(None=Constant3Vector() {constant=<rgba 0 0 -1 0>}) {transformSourceType=TRANSFORMSOURCE_Camera}
  n11 = Transform(None=Constant3Vector() {constant=<rgba -1 0 0 0>}) {transformSourceType=TRANSFORMSOURCE_Camera}
  n12 = NamedRerouteUsage()
  n13 = TransformPosition(None=NamedRerouteUsage()) {transformSourceType=TRANSFORMPOSSOURCE_Instance, transformType=TRANSFORMPOSSOURCE_World}
  n14 = Multiply(
   A=MaterialFunctionCall() {materialFunction=&ObjectScale}.Scale Z
   B=Constant() {r=1}
  )
  n15 = MaterialFunctionCall(
   UVs=MaterialFunctionCall(Plane Normal Axis=@n10, Plane X Axis=@n11, Plane Center=Add(A=@n12, B=@n13), UV Scale=@n14) {materialFunction=&VirtualPlaneCoordinates}.UVs
   Texture Scale=ScalarParameter() {defaultValue=702.2277, group=Billboard, parameterName=VirtualPlaneUVSclae}
  ) {materialFunction=&ScaleUVsByCenter}
  n16 = TextureSampleParameter2D(UVs=@n15.UVs) {
   group=Textures
   parameterName=Internal Lines Texture
   samplerType=SAMPLERTYPE_Masks
   texture=&T_WhiteMask
  }
  n17 = ComponentMask(None=@n16.RGB) {r=true}
  Ceil(None=Subtract(A=@n17, B=ScalarParameter() {defaultValue=0.6928, parameterName=LineThickness}))
  n18 = ScalarParameter() {defaultValue=1.0161, group=Hedge, parameterName=Hedge_QuantizeNormalsSteps}
  n19 = Multiply(A=VertexNormalWS(), B=@n18)
  n20 = Normalize(VectorInput=Divide(A=Round(None=@n19), B=@n18))
  n21 = CollectionParameter() {
   collection=&FortniteMaterialParameters
   parameterName=SunlightDirectionalVector
  }
  n22 = Normalize(VectorInput=Multiply(A=ComponentMask(None=@n21) {b=true, g=true, r=true}, B=Constant3Vector() {constant=<rgba 1 1 0 0>}))
  n23 = DotProduct(A=@n20, B=@n22)
  Ceil(None=@n23)
  n24 = Constant3Vector() {constant=<rgba 0 0 1 0>}
  n25 = DotProduct(A=@n20, B=Normalize(VectorInput=CrossProduct(A=@n24, B=@n22)))
  Ceil(None=@n25)
  Constant3Vector() {constant=<rgba 1 1 0 0>}
  Distance()
  n26 = MaterialFunctionCall(Float3=TransformPosition(None=Constant3Vector()) {transformSourceType=TRANSFORMPOSSOURCE_Instance, transformType=TRANSFORMPOSSOURCE_World}) {materialFunction=&BreakOutFloat3Components}
  n27 = Constant() {r=500}
  n28 = Constant() {r=1100}
  n29 = Add(A=Sine(None=Divide(A=@n26.R, B=@n27)), B=Sine(None=Divide(A=@n26.G, B=@n28)))
  Divide(A=@n29)
  n30 = ScalarParameter() {defaultValue=-5.2303, parameterName=Contrast_EdgeBillboardMask_1}
  n31 = ScalarParameter() {defaultValue=3.1093, parameterName=Contrast_EdgeBillboardMask_2}
  n32 = DotProduct(A=VertexNormalWS(), B=CameraVectorWS())
  LinearInterpolate(B=@n17, Alpha=Saturate(None=LinearInterpolate(A=@n30, B=@n31, Alpha=@n32))) {constA=1}
  n33 = Constant() {r=2}
  MaterialFunctionCall(
   "Animation  Phase (0-1)"=Constant()
   Number of Rows=@n33
   Number of Columns=@n33
   Texture=TextureObject() {samplerType=SAMPLERTYPE_Masks, texture=&T_WhiteMask}
   UVs=@n15.UVs
  ) {materialFunction=&FlipBook}
  n34 = NamedRerouteUsage()
  n35 = Subtract(A=CameraPositionWS(), B=@n34)
  MaterialFunctionCall(Vector=@n35) {materialFunction=&SafeNormalize}
  n36 = CollectionParameter() {
   collection=&FortniteMaterialParameters
   parameterName=SunlightDirectionalVector
  }
  n37 = ShadowReplace(Default=Normalize(VectorInput=@n35), Shadow=ComponentMask(None=@n36) {b=true, g=true, r=true})
  n38 = Normalize(VectorInput=VertexNormalWS())
  n39 = DotProduct(A=@n37, B=@n38)
  n40 = StaticSwitchParameter(True=Multiply(A=@n37, B=Sign(None=@n39)), False=@n37) {defaultValue=true, group=Billboard, parameterName=BillBoardShortestPathRotation}
  n41 = CrossProduct(A=@n38, B=@n40)
  MaterialFunctionCall(Vector=@n41) {materialFunction=&SafeNormalize}
  n42 = WorldPosition()
  MaterialFunctionCall(VectorToTransform=@n42.XYZ, BasisPosition=@n34) {materialFunction=&Transform3x3Matrix}
  n43 = StaticBool() {desc=Don't change}
  n44 = Constant() {r=1}
  n45 = NamedRerouteUsage()
  n46 = NamedRerouteUsage()
  n47 = StaticSwitchParameter(
   True=NamedRerouteUsage()
   False=Multiply(
    A=StaticSwitchParameter(True=@n45, False=@n46) {group=VirtualPlaneShading, parameterName=UseVirtualPlaneShading, sortPriority=1}
    B=NamedRerouteUsage()
   )
  ) {group=Hedge, parameterName=HedgeShading, sortPriority=1}
  n48 = NamedRerouteUsage()
  n49 = MaterialFunctionCall(Float3=NamedRerouteUsage()) {materialFunction=&BreakOutFloat3Components}
  n50 = Add(A=Add(A=@n49.R, B=@n49.R), B=@n49.B)
  n51 = ScalarParameter() {defaultValue=33.3, group=Outlines, parameterName=ClusterRandomOutline}
  n52 = Add(A=@n48, B=Frac(None=Divide(A=@n50, B=@n51) {constB=10}))
  n53 = LinearInterpolate(
   A=Constant() {r=0.05}
   B=Constant() {r=0.49}
   Alpha=StaticSwitchParameter(True=Frac(None=@n52), False=@n48) {group=Outlines, parameterName=OutlineEachCluster}
  )
  n54 = TextureSampleParameter2D() {group=Textures, parameterName=OpacityMask, samplerType=SAMPLERTYPE_Masks, texture=&T_WhiteMask}
  n55 = SetMaterialAttributes!(
   MaterialAttributes
   Base Color=StaticSwitch(True=NamedRerouteUsage(), False=@n47, Value=NamedRerouteUsage())
   Specular=Constant()
   Roughness=@n53
   Opacity=Constant()
   Subsurface Color=Constant3Vector()
   Normal=StaticSwitch(True=NamedRerouteUsage(), False=NamedRerouteUsage(), Value=NamedRerouteUsage())
   Opacity Mask=@n54.R
  ) {attributeSetTypes [
    #1
    #2
    #3
    #4
    #5
    #6
    #7
   ]}
  n56 = StaticSwitchParameter(True=NamedRerouteUsage(), False=NamedRerouteUsage()) {group=Billboard, parameterName=UseFallbackBillboard}
  n57 = MaterialFunctionCall!(In=StaticSwitchParameter(
    True=SetMaterialAttributes!(MaterialAttributes=@n55, World Position Offset=@n56) {attributeSetTypes [#8]}
    False=@n55
   ) {defaultValue=true, group=Billboard, parameterName=Enable_Billboard}) {materialFunction=/LimeRock_Tree/Assets/Treehouse/MaterialFunctions/MF_LimeRock_WE_Trunk}
  n58 = MaterialFunctionCall!(In=@n57.Result, OpaqueCanopyMask) {materialFunction=/LimeRock_Tree/Assets/Treehouse/MaterialFunctions/MF_LimeRock_WE_ApolloTreeAnim}
  n59 = MaterialFunctionCall!(In=@n57.Result) {materialFunction=/LimeRock_Tree/Assets/Treehouse/MaterialFunctions/MF_WildEstate_Tree_Animation_2}
  n60 = MaterialFunctionCall!(In=StaticSwitchParameter(True=@n58.Result, False=@n59.Result) {parameterName=UseFallBackAnimation}) {materialFunction=/LimeRock_Tree/Assets/Treehouse/MaterialFunctions/MF_LimeRock_UnlitShading}
  n61 = StaticSwitchParameter(
   True=SetMaterialAttributes(MaterialAttributes=@n60.Result, World Position Offset=Constant3Vector()) {attributeSetTypes [#8]}
   False=@n60.Result
  ) {parameterName=DisableWPO}
  n62 = StaticSwitch(
   True=SetMaterialAttributes(
    MaterialAttributes=MaterialFunctionCall!(Material Attributes=@n61, IsTangentSpaceNormal=@n43, In=@n44) {materialFunction=/Game/ImpostorBaker/MaterialFunctions/ImposterCaptureSwitch}.Result
    World Position Offset=Constant3Vector()
   ) {attributeSetTypes [#8]}
   False=@n61
   Value=NamedRerouteDeclaration(None=StaticBool()) {name=IsImpostorBake, nodeColor=<rgba 0 1 1 1>}
  )
  n63 = GetMaterialAttributes(None=@n62) {attributeGetTypes [#8]}
  n64 = PixelDepth()
  n65 = ScalarParameter() {defaultValue=7500, group=WPO Fade, parameterName=WPO_Fade_Offset}
  n66 = ScalarParameter() {defaultValue=1500, group=WPO Fade, parameterName=WPO_Fade_Distance}
  n67 = NaniteReplace!(
   Default=Divide(A=Subtract(A=@n64.R, B=@n65), B=@n66)
   Nanite=MaterialFunctionCall!(FadeBlendDistance=@n66, AutoHandleBadDistance) {materialFunction=/Game/Global/MaterialFunctions/TechArt/MF_WPODisableBlend}.VertexShaderOutput
  )
  n68 = SetMaterialAttributes(
   MaterialAttributes=@n63.MaterialAttributes
   World Position Offset=LinearInterpolate(A=@n63.World Position Offset, B=Constant3Vector(), Alpha=Saturate(None=@n67))
  ) {attributeSetTypes [#8]}
  MaterialFunctionCall!(
   In=MaterialFunctionCall(
    Material_In=StaticSwitchParameter(True=@n68, False=@n62) {defaultValue=true, group=WPO Fade, parameterName=UseWPOFade}
    UseCrackedDamageTextures=StaticBool()
    UseAdditiveWPO=StaticBool() {value=true}
   ) {materialFunction=&MF_FN_DestructibleDamage}.Material_Out
   DestructibleDamageAmount
  ) {materialFunction=/LimeRock_Tree/Assets/Treehouse/MaterialFunctions/MF_LimeRock_Cracks}
  Multiply(A=Multiply())
  NamedRerouteDeclaration(None=StaticSwitchParameter(True=@n58.Leaf_Pivot_LS, False=@n59.LeafPivot_LocalSpace) {parameterName=UseFallBackAnimation}) {name=LeafPivot_LocalSpace, nodeColor=<rgba 0.8047 0 1 1>}
  n69 = Multiply(A=ScalarParameter() {defaultValue=-0.99, group=BillBoard WPO, parameterName=BillboardSqueze}, B=Multiply(A=@n7, B=@n8) {constB=-1})
  n70 = Multiply(A=@n7, B=ScalarParameter() {defaultValue=-100, group=BillBoard WPO, parameterName=BillboardOffsetCamDir})
  NamedRerouteDeclaration(None=ShadowReplace(Default=Add(A=@n69, B=@n70), Shadow=Constant3Vector())) {name=WPO_LeafFlatten, nodeColor=<rgba 0 1 0.4844 1>}
  NamedRerouteDeclaration(None=StaticSwitchParameter(True=@n58.Anim WPO_Only, False=@n59.Anim_WPO_Only) {parameterName=UseFallBackAnimation}) {name=Anim_WPO_only, nodeColor=<rgba 0.8047 0 1 1>}
  NamedRerouteDeclaration(None=VertexColor().R) {name=PerLeafRandom, nodeColor=<rgba 1 0 0.6308 1>}
  NamedRerouteDeclaration(None=Divide(A=Add(A=@n29) {constB=2}) {constB=4}) {name=PerActorRandom, nodeColor=<rgba 0 1 0.7891 1>}
  n71 = NamedRerouteUsage()
  n72 = ScalarParameter() {parameterName=InteriorLines_MinOpacity}
  NamedRerouteDeclaration(None=LinearInterpolate(B=@n17, Alpha=Max(A=@n71, B=@n72)) {constA=1}) {name=Interior Lines Virtual PlaneBillboard, nodeColor=<rgba 1 0 0.3984 1>}
  NamedRerouteDeclaration(None=@n59.OveraAllBend) {name=Anim_WPO_OverallBend, nodeColor=<rgba 0.8047 0 1 1>}
  n73 = Subtract(A=PixelDepth().R, B=ScalarParameter() {defaultValue=400, group=Outlines, parameterName=InteriorLines_Fade_Offset})
  n74 = ScalarParameter() {defaultValue=1600, group=Outlines, parameterName=InteriorLines_Fade_Distance}
  NamedRerouteDeclaration(None=Saturate(None=Divide(A=@n73, B=@n74))) {name=NearFarMask, nodeColor=<rgba 0 1 0.2031 1>}
  n75 = DotProduct(A=@n40, B=@n38)
  n76 = MaterialFunctionCall(Multiplier=Constant() {r=0.5}) {materialFunction=&Pi}
  n77 = RotateAboutAxis(
   NormalizedRotationAxis=Normalize(VectorInput=@n41)
   RotationAngle=Multiply(A=Arccosine!(None=@n75), B=Divide(A=@n76.Result) {constB=10})
   PivotPoint=@n34
   Position=@n42.XYZ
  )
  NamedRerouteDeclaration(None=Add(A=NamedRerouteUsage(), B=NamedRerouteDeclaration(None=@n77) {name=BillboardTransformFallback, nodeColor=<rgba 1 0.1641 0 1>})) {name=WPO BillboardTransformFallback, nodeColor=<rgba 1 0.1641 0 1>}
  n78 = Multiply(
   A=ScalarParameter() {defaultValue=0.104, group=Color Variation WorldSpace, parameterName=RandomHueShift}
   B=Subtract(A=NamedRerouteUsage()) {constB=0.5}
  )
  n79 = VectorParameter() {group=Color, parameterName=Color, defaultValue=<rgba 0.0061 0.1354 0.0559 1>}
  n80 = TextureSampleParameter2D() {
   group=Textures
   parameterName=Diffuse
   samplerSource=SSM_Wrap_WorldGroupSettings
   texture=/Game/Characters/Fortress_Character/Textures/T_All_White
  }
  n81 = StaticSwitchParameter(True=Multiply(A=@n79.RGB, B=@n80.RGB), False=@n79.RGB) {group=Color, parameterName=EnableDiffuseTexture}
  NamedRerouteDeclaration(None=StaticSwitchParameter(
    True=MaterialFunctionCall(Hue Shift Percentage=@n78, Texture=@n81) {materialFunction=&HueShift}.Result
    False=@n81
   ) {defaultValue=true, group=Color Variation WorldSpace, parameterName=UseRandomHueColor}) {name=ColorBase, nodeColor=<rgba 0 1 0.1562 1>}
  n82 = CollectionParameter() {
   collection=&FortniteMaterialParameters
   parameterName=SunlightDirectionalVector
  }
  n83 = ComponentMask(None=@n82) {b=true, g=true, r=true}
  n84 = CameraVectorWS()
  NamedRerouteDeclaration(None=Normalize(VectorInput=Add(A=@n83, B=@n84) {desc=Compromise for Direction light and all othe rlights})) {name=NormalsFinal, nodeColor=<rgba 0 1 0.4609 1>}
  n85 = ScalarParameter() {defaultValue=0.764, group=VirtualPlaneShading, parameterName=DirectionalShadow Min Darkness}
  n86 = WorldPosition()
  n87 = ObjectPositionWS()
  n88 = TextureSampleParameter2D(UVs=MaterialFunctionCall(
    UVs=MaterialFunctionCall!(Plane Normal Axis=@n10, Plane X Axis=@n11, WorldPosition=@n86.XYZ, Plane Center=@n87, UV Scale=@n14) {materialFunction=/LimeRock_Tree/Assets/Treehouse/MaterialFunctions/LimeRockVirtualPlaneCoordinates}.UVs
    Texture Scale=ScalarParameter() {defaultValue=655.1484, group=VirtualPlaneShading, parameterName=VirtualPlane UV_Scale}
   ) {materialFunction=&ScaleUVsByCenter}.UVs) {
   group=VirtualPlaneShading
   parameterName=VirtualPlaneHightMap
   texture=/LimeRock_Tree/Assets/Treehouse/Textures/T_WildEstate_Tree_Noise_B_M
   channelNames {a="", b=HeightSoftMask, g=HeightHardMask, r=""}
  }
  n89 = ScalarParameter() {defaultValue=3, group=VirtualPlaneShading, parameterName=StepsB2}
  n90 = Divide(A=Round(None=Multiply(A=@n88.HeightHardMask, B=@n89)), B=@n89)
  n91 = ScalarParameter() {defaultValue=0.2606, group=VirtualPlaneShading, parameterName=Heightmap_Strength_Hard}
  n92 = Multiply(A=@n88.HeightSoftMask, B=ScalarParameter() {defaultValue=0.2606, group=VirtualPlaneShading, parameterName=Heightmap_Strength_Soft})
  n93 = WorldPosition()
  n94 = ObjectPositionWS()
  n95 = CollectionParameter() {
   collection=&FortniteMaterialParameters
   parameterName=SunlightDirectionalVector
  }
  n96 = DotProduct(A=Normalize(VectorInput=Subtract(A=@n93.XYZ, B=@n94)), B=ComponentMask(None=@n95) {b=true, g=true, r=true})
  n97 = ScalarParameter() {defaultValue=2, group=VirtualPlaneShading, parameterName=StepsB}
  n98 = Round(None=Multiply(A=Add(A=@n92, B=@n96), B=@n97))
  n99 = Add(
   A=Add(A=Multiply(A=@n90, B=@n91), B=Divide(A=@n98, B=@n97))
   B=ScalarParameter() {defaultValue=-0.6278, group=VirtualPlaneShading, parameterName=Threshold Shadowing}
  )
  NamedRerouteDeclaration(None=Saturate(None=LinearInterpolate(A=@n85, Alpha=@n99))) {name=DotVirtualPlaneShading, nodeColor=<rgba 1 0 0.5156 1>}
  n100 = ScalarParameter() {defaultValue=0.512, group=Color, parameterName=Shading_min}
  n101 = ScalarParameter() {defaultValue=1.1151, group=Color, parameterName=Shading_Max}
  n102 = WorldPosition()
  n103 = ObjectPositionWS()
  n104 = CollectionParameter() {
   collection=&FortniteMaterialParameters
   parameterName=SunlightDirectionalVector
  }
  n105 = DotProduct(A=Normalize(VectorInput=Subtract(A=@n102.XYZ, B=@n103)), B=ComponentMask(None=@n104) {b=true, g=true, r=true})
  n106 = ScalarParameter() {defaultValue=2, group=Color, parameterName=DualShading_Range, sortPriority=99}
  n107 = LocalPosition()
  n108 = ComponentMask(None=PreSkinnedLocalBounds().Max) {b=true}
  n109 = Add(
   A=VectorParameter() {group=Color, parameterName=Shading Center LS, defaultValue=<rgba 0 0 0 1>}.RGB
   B=AppendVector(A=Constant2Vector(), B=Divide(A=@n108))
  )
  n110 = Divide(A=Length(None=Subtract(A=@n107.XYZ, B=@n109)) {is=class MaterialExpressionLength}) {constB=100}
  n111 = ScalarParameter() {defaultValue=0.1, group=Color, parameterName=DualShading_Leaf_Length_Contribution, sortPriority=99}
  n112 = NamedRerouteUsage()
  n113 = NamedRerouteUsage()
  n114 = Subtract(
   A=VertexInterpolator(VS=StaticSwitchParameter(True=@n112, False=@n113) {group=Color, parameterName=UseLeafClusterMergeForShading}).PS
   B=@n109
  )
  n115 = VectorParameter() {group=Color, parameterName=Shading Scale, defaultValue=<rgba 1 1 1 1>}
  n116 = Normalize(VectorInput=Transform(None=Multiply(A=@n114, B=@n115.RGB)) {transformSourceType=TRANSFORMSOURCE_Local})
  n117 = CollectionParameter() {
   collection=&FortniteMaterialParameters
   parameterName=SunlightDirectionalVector
  }
  n118 = DotProduct(A=@n116, B=ComponentMask(None=@n117) {b=true, g=true, r=true})
  n119 = Add(
   A=Add(A=Divide(A=@n105, B=@n106), B=Multiply(A=@n110, B=@n111))
   B=Multiply(A=@n118, B=ScalarParameter() {defaultValue=0.1, group=Color, parameterName=DualShading_Leaf_Contribution, sortPriority=99})
  )
  n120 = ScalarParameter() {defaultValue=-0.5, group=Color, parameterName=DualShading_Bias, sortPriority=99}
  n121 = ScalarParameter() {defaultValue=3, group=Color, parameterName=DualShading_Steps, sortPriority=99}
  n122 = Round(None=Multiply(A=Add(A=@n119, B=@n120), B=@n121))
  n123 = ScalarParameter() {defaultValue=1, desc="1 is default and even amount of light and shade on the tree. ", group=Color, parameterName=Shading_Bias}
  n124 = ScalarParameter() {
   defaultValue=2
   desc="Shading Range on the Dot product 2 = full range. From -1 to 1 of the dot product.\r\nlower value makes the shaded area smaller \r\nhigher value makes the shaded area larger"
   group=Color
   parameterName=Shading_Range
  }
  n125 = ScalarParameter() {
   defaultValue=8
   desc="Shading Range on the Dot product 2 = full range. From -1 to 1 of the dot product.\r\nlower value makes the shaded area smaller \r\nhigher value makes the shaded area larger"
   group=Color
   parameterName=Shading_Steps
  }
  n126 = Multiply(A=Divide(A=Add(A=@n118, B=@n123), B=@n124), B=@n125)
  n127 = StaticSwitchParameter(True=Saturate(None=Divide(A=@n122, B=@n121)), False=Divide(A=Round(None=@n126), B=@n125)) {group=Color, parameterName=UseDualShading, sortPriority=98}
  NamedRerouteDeclaration(None=Saturate(None=LinearInterpolate(A=@n100, B=@n101, Alpha=@n127))) {name=MergedLeafDotShading, nodeColor=<rgba 1 0.1641 0 1>}
  NamedRerouteDeclaration(None=@n116) {name=NormalsForImpostor, nodeColor=<rgba 1 0 0.2344 1>}
  NamedRerouteDeclaration(None=@n81) {name=ColorBase_ForImpostorBake, nodeColor=<rgba 0 1 0.1562 1>}
  NamedRerouteDeclaration(None=MaterialFunctionCall(Number=VertexInterpolator(VS=@n108).PS, UVs=ScreenPosition().ViewportUV) {materialFunction=&DebugScalarValues}.Result) {name=DebugTest, nodeColor=<rgba 0.5234 0 1 1>}
  NamedRerouteDeclaration(None=@n58.Leaf_Pivot_WS) {name=LeafPivot_FallBack_WorldSpace, nodeColor=<rgba 0.8047 0 1 1>}
  NamedRerouteDeclaration(None=@n58.Leaf_XAxis_WS) {name=Leaf XAxis_ WorldSpace, nodeColor=<rgba 0.8047 0 1 1>}
  n128 = Saturate(None=Sign(None=@n25))
  n129 = Sign(None=@n23)
  n130 = Multiply(
   A=LinearInterpolate(A=ScalarParameter() {defaultValue=0.88, group=Hedge, parameterName=HedgeShadingMidShade}, Alpha=Saturate(None=@n128))
   B=LinearInterpolate(A=ScalarParameter() {defaultValue=0.75, group=Hedge, parameterName=HedgeShadingShadow}, Alpha=Saturate(None=@n129))
  )
  n131 = ScalarParameter() {defaultValue=1.25, group=Hedge, parameterName=HedgeShadingTopShade}
  n132 = DotProduct(A=@n20, B=@n24)
  n133 = Saturate(None=Sign(None=@n132))
  NamedRerouteDeclaration(None=Multiply(A=LinearInterpolate(A=@n130, B=@n131, Alpha=@n133) {constA=1}, B=NamedRerouteUsage())) {name=HedgeShading, nodeColor=<rgba 0.4297 0 1 1>}
  NamedRerouteUsage()
  NamedRerouteUsage()
  NamedRerouteUsage()
  NamedRerouteUsage()
  ObjectLocalBounds()
  ObjectPositionWS()
  ObjectRadius()
  Round(None=@n132)
  Round(None=@n23)
  n134 = Multiply(A=@n8) {constB=-1}
  n135 = ScalarParameter() {defaultValue=104.7052, group=BillBoard WPO, parameterName=Billboard_EdgeMask}
  Saturate(None=Divide(A=Subtract(A=@n134, B=@n135)) {constB=0.0001})
  ScalarParameter() {defaultValue=1000, group=Color Variation WorldSpace, parameterName=Color Variation Distance WS, sortPriority=10}
  ScalarParameter() {defaultValue=1100, group=Color Variation WorldSpace, parameterName=Color Variation Distance WS2, sortPriority=10}
  n136 = WorldPosition()
  SetMaterialAttributes(
   MaterialAttributes=@n62
   World Position Offset=Multiply(A=Subtract(B=@n136.XYZ), B=ScalarParameter() {defaultValue=0.896, parameterName=Testeste})
  ) {attributeSetTypes [#8]}
  ShadowReplace(Default=@n77, Shadow=Constant3Vector())
  n137 = Multiply(A=@n7, B=Saturate(None=@n9))
  n138 = ScalarParameter() {defaultValue=-206.2373, group=BillBoard WPO, parameterName=BillboardSqueze_PushInner}
  StaticSwitchParameter(True=@n69, False=Add(A=@n69, B=Multiply(A=@n137, B=@n138))) {defaultValue=true, parameterName=UseVirtual Plane Card}
  StaticSwitchParameter(True=TwoSidedSign(), False=@n44) {parameterName=UseFlipBackFaceNormals}
  StaticSwitchParameter(True=Constant() {r=0.678}, False=@n53) {group=Hedge, parameterName=HedgeShading, sortPriority=1}
  StaticSwitchParameter(True=Multiply(A=NamedRerouteUsage(), B=@n47), False=@n47) {defaultValue=true, parameterName=UseVirtual Plane Card}
  n139 = TextureCoordinate() {coordinateIndex=2}
  n140 = AppendVector(
   A=ComponentMask(None=TextureCoordinate() {coordinateIndex=1}) {r=true}
   B=AppendVector(A=ComponentMask(None=@n139) {r=true}, B=ComponentMask(None=@n139) {g=true})
  )
  Subtract(A=Multiply(A=Frac(None=@n140)) {constB=12000}, B=Constant3Vector() {constant=<rgba 6000 6000 3000 0>}) {constB=6000}
  TransformPosition(None=NamedRerouteDeclaration(None=NamedRerouteUsage()) {name=LeafPivot Position LS, nodeColor=<rgba 0.0781 1 0 1>}) {transformType=TRANSFORMPOSSOURCE_World}
  Transform(None=@n3) {transformSourceType=TRANSFORMSOURCE_World, transformType=TRANSFORM_Camera}
  WorldPosition()
 ]
}
